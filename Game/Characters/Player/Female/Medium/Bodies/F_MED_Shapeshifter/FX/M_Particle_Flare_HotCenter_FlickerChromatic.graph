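# Additive unlit Niagara sprite material for a flickering flare with a hot center and chromatic fringing, in the Shapeshifter character's FX folder.
# A radial mask builds a desaturated bright core and soft edge, adding three offset flare-texture samples as RGB to split colors.
# Particle color tints it, and dynamic parameters drive hue shift and multiplied-sine flicker.
# Particle alpha, depth fade (skipped on mobile) and near-camera fade also attenuate it.
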Material /Game/Characters/Player/Female/Medium/Bodies/F_MED_Shapeshifter/FX/M_Particle_Flare_HotCenter_FlickerChromatic
graph {
 properties {
  bUsedWithNiagaraSprites=true
  blendMode=BLEND_Additive
  ditherOpacityMask=true
  pixelDepthOffsetMode=PDOM_Legacy
  shadingModel=MSM_Unlit
  usedShadingModels=Unlit
 }
 nodes [
  n1 = Constant() {r=0.5}
  n2 = SphereMask(A=TextureCoordinate(), B=@n1, Radius=@n1) {hardnessPercent=0}
  n3 = Multiply(
   A=ScalarParameter() {defaultValue=2, parameterName=CenterBrightness}
   B=Power(Base=@n2, Exp=ScalarParameter() {defaultValue=25, parameterName=CentertPower}) {constExponent=25}
  ) {constB=2}
  n4 = ParticleColor()
  n5 = Desaturation(None=@n4.RGB, Fraction=ScalarParameter() {defaultValue=0.3, parameterName=CenterDesaturation})
  n6 = ScalarParameter() {defaultValue=4, parameterName=EdgePower}
  n7 = TextureCoordinate()
  n8 = ComponentMask(None=VectorParameter() {parameterName=Chromatic Offset, defaultValue=<rgba 0.0125 0.0125 0 0>}.RGB) {g=true, r=true}
  n9 = Add(A=@n7, B=@n8)
  n10 = Add(A=@n7, B=Multiply(A=@n8) {constB=-1})
  n11 = Add(
   A=Multiply(A=Power(Base=@n2, Exp=@n6) {constExponent=4}, B=ScalarParameter() {defaultValue=0.15, parameterName=EdgeBrightness}) {constB=0.15}
   B=MaterialFunctionCall(
    R=TextureSampleParameter2D(UVs=@n9) {parameterName=FlareMask, texture=/Game/Effects/Fort_Effects/Textures/Flares/T_Flare_Spikeball}.R
    G=TextureSampleParameter2D(UVs=@n7) {parameterName=FlareMask, texture=&T_Flare_Spikeball}.R
    B=TextureSampleParameter2D(UVs=@n10) {parameterName=FlareMask, texture=&T_Flare_Spikeball}.R
   ) {materialFunction=&AppendMany}.RGB
  )
  n12 = MaterialFunctionCall(
   Hue Shift Percentage=DynamicParameter() {paramNames [FlickerFade,FlickerSpeed,ColorJitter,Param4]}.ColorJitter
   Texture=Add(A=Multiply(A=@n3, B=@n5), B=Multiply(A=@n4.RGB, B=@n11))
  ) {materialFunction=&HueShift}
  MaterialFunctionCall(Emissive=MaterialFunctionCall(LDR Input=Saturate(None=@n12.Result), HDR Tint=@n12.Result) {materialFunction=&DeriveHDRfromLDR}.RemappedRange) {materialFunction=&MF_FX_EmissiveAdjustments}
  n13 = DynamicParameter() {paramNames [FlickerFade,FlickerSpeed,ColorJitter,Param4]}
  n14 = Multiply(A=Time(), B=@n13.FlickerSpeed)
  n15 = Sine(None=@n14) {period=0.075}
  n16 = Multiply(A=Multiply(A=Sine(None=@n14) {period=0.155}, B=Sine(None=@n14) {period=0.2}), B=Sine(None=@n14) {period=0.43})
  n17 = LinearInterpolate(A=Clamp(None=Multiply(A=@n15, B=@n16)) {maxDefault=3}, Alpha=Saturate(None=@n13.FlickerFade)) {constA=1}
  n18 = ParticleColor()
  n19 = DynamicParameter() {desc=Index 1 Params for Niagara, parameterIndex=1, defaultValue=<rgba 30 1 1 1>, paramNames [DepthFadeDist, Param2, Param3, Param4]}
  n20 = ScalarParameter() {defaultValue=30, parameterName=DepthFade}
  n21 = DepthFade(Opacity=@n18.A, FadeDistance=If(A=@n19.DepthFadeDist, B=Constant() {r=1}, "A > B"=@n19.DepthFadeDist, "A == B"=@n20, "A < B"=@n20))
  Multiply(
   A=Multiply(A=@n17, B=ShadingPathSwitch(Default=@n21, Mobile=@n18.A))
   B=MaterialFunctionCall(Distance=ScalarParameter() {defaultValue=250, parameterName=CameraFadeDistance}) {materialFunction=&Near_Camera_Fade}.Result
  )
  Multiply(A=Desaturation(None=@n12.Result), B=Multiply(A=@n17, B=@n18.A))
 ]
}
