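# Additive unlit caustic effect material for the Shapeshifter character, usable on Niagara sprites and skeletal meshes.
# Three offset samples of one caustic texture form RGB chromatic fringing, scaled and blended by dynamic parameters, then tinted by particle color for HDR emissive.
# Desaturated caustics drive erosion, multiplied by particle alpha, a radial mask and two depth fades, one inverted to favor surface intersections.
# An optional world-space sphere mask fades the result around object or particle position.
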Material /Game/Characters/Player/Female/Medium/Bodies/F_MED_Shapeshifter/FX/M_Shapeshifter_CausticMagic
graph {
 properties {
  bUsedWithMeshDeformer=true
  bUsedWithNiagaraSprites=true
  bUsedWithSkeletalMesh=true
  blendMode=BLEND_Additive
  pixelDepthOffsetMode=PDOM_Legacy
  shadingModel=MSM_Unlit
  usedShadingModels=Unlit
 }
 nodes [
  n1 = DynamicParameter() {paramNames [Change,UScale,VScale,erode]}
  n2 = DynamicParameter() {paramNames [Change,UScale,VScale,erode]}
  n3 = Multiply(A=TextureCoordinate(), B=AppendVector(A=@n2.UScale, B=@n2.UScale))
  n4 = ComponentMask(None=VectorParameter() {parameterName=ChromaOffset, defaultValue=<rgba 0.01 0.01 1 1>}.RGB) {g=true, r=true}
  n5 = Panner(Coordinate=Add(A=@n3, B=@n4))
  n6 = Panner(Coordinate=@n3)
  n7 = Panner(Coordinate=Add(A=@n3, B=Multiply(A=@n4) {constB=-1}))
  n8 = LinearInterpolate(
   B=MaterialFunctionCall(
    X=TextureSampleParameter2D(UVs=@n5) {parameterName=Caustic, texture=/Game/Global/Textures/Default/Caustics/T_Caustics_Softer}.R
    Y=TextureSampleParameter2D(UVs=@n6) {parameterName=Caustic, texture=&T_Caustics_Softer}.R
    Z=TextureSampleParameter2D(UVs=@n7) {parameterName=Caustic, texture=&T_Caustics_Softer}.R
   ) {materialFunction=&MakeFloat3}.Result
   Alpha=@n1.Change
  ) {constA=1}
  n9 = Desaturation(None=@n8)
  n10 = MaterialFunctionCall!(Dynamic Param=@n1.erode, Alpha=@n9) {materialFunction=/Game/Effects/Fort_Effects/Materials/_Functions/Erode_Alpha}
  n11 = ParticleColor()
  n12 = Constant() {r=0.5}
  n13 = Multiply(A=@n11.A, B=SphereMask(A=TextureCoordinate(), B=@n12, Radius=@n12, Hardness=ScalarParameter() {parameterName=Hardness}))
  n14 = OneMinus(None=DepthFade(FadeDistance=ScalarParameter() {defaultValue=15, parameterName=Fade Distancve}))
  n15 = ScalarParameter() {defaultValue=25, parameterName=TruDepthFadeValue}
  n16 = Multiply(A=@n10.Result, B=Multiply(A=Multiply(A=@n13, B=@n14), B=DepthFade(FadeDistance=@n15)))
  n17 = WorldPosition()
  n18 = LinearInterpolate(A=ObjectPositionWS(), B=ParticlePositionWS(), Alpha=ScalarParameter() {parameterName=Use Particle Position Center})
  n19 = DynamicParameter() {parameterIndex=1, defaultValue=<rgba 500 1 1 1>, paramNames [Radius,Param2,Param3,Param4]}
  LinearInterpolate(
   A=@n16
   B=Multiply(A=@n16, B=SphereMask(A=@n17.XYZ, B=@n18, Radius=@n19.Radius) {hardnessPercent=0})
   Alpha=ScalarParameter() {parameterName=Use SphereMask FadeOut}
  )
  n20 = Multiply(A=@n8, B=@n11.RGB)
  MaterialFunctionCall(Emissive=MaterialFunctionCall(LDR Input=Saturate(None=@n20), HDR Tint=@n20) {materialFunction=&DeriveHDRfromLDR}.RemappedRange) {materialFunction=&MF_FX_EmissiveAdjustments}
  n21 = Multiply(A=@n10.Result, B=@n13)
  n22 = VectorParameter() {parameterName=Dither Offsets, defaultValue=<rgba 1 0.2 0.3 0.7>}
  MaterialFunctionCall(Alpha Threshold=Saturate(None=Multiply(A=@n9, B=@n21)), Offset=AppendVector(A=@n22.R, B=@n22.R), Offset2=AppendVector(A=@n22.B, B=@n22.B)) {materialFunction=&DitherTemporalAA_Fortnite}
  SceneTexture() {sceneTextureId=PPI_WorldNormal}
 ]
}
