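# Parent material for the Shapeshifter character's crystal body, layering animated faux-depth crystal over desaturated diffuse and a height/Fresnel gradient.
# Pre-skinned positions, projected by normal, drive bump-offset panning noise for sparkles and inner clouds, colored through a curve atlas with faked shimmering reflections.
# Fresnel and a fake view-space highlight shape that glow, which feeds base color and emissive.
# Switches select normal blending and channel-mask modulation before the shared character PostFX function.
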Material /Game/Characters/Player/Female/Medium/Bodies/F_MED_Shapeshifter/Materials/M_Crystal_Parent
graph {
 properties {
  bUseMaterialAttributes=true
  bUsedWithMeshDeformer=true
  bUsedWithMorphTargets=true
  bUsedWithNanite=true
  bUsedWithSkeletalMesh=true
  pixelDepthOffsetMode=PDOM_Legacy
  usedShadingModels=Default Lit
 }
 nodes [
  n1 = TextureSampleParameter2D() {
   group=CrystalSuperleveling
   parameterName=M
   samplerType=SAMPLERTYPE_Masks
   texture=/Game/Characters/Player/Female/Medium/Bodies/F_MED_Shapeshifter/Textures/T_F_MED_Shapeshifter_Body_M
  }
  n2 = MaterialFunctionCall() {materialFunction=&BaseFN_CharacterFunction_PreFX}
  n3 = TextureSampleParameter2D() {
   group=CrystalSuperleveling
   parameterName=Diffuse
   texture=/Game/Characters/Player/Female/Medium/Bodies/F_MED_Shapeshifter/Textures/T_F_MED_Shapeshifter_Body_D
  }
  n4 = ScalarParameter() {defaultValue=0.25, group=CrystalSuperleveling, parameterName=BaseColorDesaturationFraction}
  n5 = Multiply(
   A=Add(A=Desaturation(None=@n3.RGB, Fraction=@n4), B=ScalarParameter() {group=CrystalSuperleveling, parameterName=BaseColorBias})
   B=ScalarParameter() {defaultValue=0.2, group=CrystalSuperleveling, parameterName=BaseColorScale}
  )
  n6 = VectorParameter() {group=CrystalSuperleveling, parameterName=BaseColorLowCrystal, defaultValue=<rgba 0.0022 0.0079 0.0451 0>}
  n7 = VectorParameter() {group=CrystalSuperleveling, parameterName=BaseColorHighCrystal, defaultValue=<rgba 0.3747 0.3965 0.9549 0>}
  n8 = VertexInterpolator(VS=PreSkinnedPosition())
  n9 = MaterialFunctionCall() {materialFunction=&ObjectLocalBounds}
  n10 = Subtract(A=@n8.PS, B=@n9.Local Bounds Minimum)
  n11 = TextureSampleParameter2D() {
   group=CrystalSuperleveling
   parameterName=Normals
   samplerType=SAMPLERTYPE_Normal
   texture=/Game/Characters/Player/Female/Medium/Bodies/F_MED_Shapeshifter/Textures/T_F_MED_Shapeshifter_Body_N
  }
  n12 = MaterialFunctionCall(
   Exponent=ScalarParameter() {defaultValue=2.8, group=CrystalSuperleveling, parameterName=FresnelExponent}
   Vertex / Pixel Normal=Transform(None=@n11.RGB)
  ) {materialFunction=&MF_Fresnel_FX}
  n13 = Min(A=ComponentMask(None=Divide(A=@n10, B=@n9.Local Bounds Size)) {b=true}, B=@n12.Fresnel Inverted)
  n14 = TextureSampleParameter2D() {
   group=CrystalSuperleveling
   parameterName=M
   samplerType=SAMPLERTYPE_Masks
   texture=&T_F_MED_Shapeshifter_Body_M
  }
  n15 = Transform(None=PixelNormalWS()) {transformSourceType=TRANSFORMSOURCE_World, transformType=TRANSFORM_View}
  n16 = Normalize(VectorInput=VectorParameter() {group=CrystalSuperleveling, parameterName=FresnelHighlightVector, defaultValue=<rgba 0.6 1 0.6 0>}.RGB)
  n17 = Multiply(
   A=ScalarParameter() {defaultValue=2, group=CrystalSuperleveling, parameterName=FakeHighlightScale}
   B=Saturate(None=DotProduct(A=@n15, B=@n16))
  )
  n18 = PreSkinnedNormal()
  n19 = LinearInterpolate(
   A=ComponentMask(None=@n8.PS) {b=true, r=true}
   B=ComponentMask(None=@n8.PS) {b=true, g=true}
   Alpha=MaterialFunctionCall(Float3=VertexInterpolator(VS=@n18).PS) {materialFunction=&BreakOutFloat3Components}.R
  )
  n20 = Panner(
   Coordinate=Divide(A=@n19, B=ScalarParameter() {defaultValue=140, group=CrystalSuperleveling, parameterName=CrystalTile_1})
   Speed=AppendVector(
    A=ScalarParameter() {group=CrystalSuperleveling, parameterName=CrystalPanSpeedU_1}
    B=ScalarParameter() {group=CrystalSuperleveling, parameterName=CrystalPanSpeedV_1}
   )
  ) {bFractionalPart=true}
  n21 = TextureSampleParameter2D() {
   group=CrystalSuperleveling
   parameterName=CrystalMap
   samplerType=SAMPLERTYPE_LinearColor
   texture=/Game/Characters/Player/Female/Medium/Bodies/F_MED_Shapeshifter/Textures/T_Shapeshifter_Crystal_Body
  }
  n22 = Multiply(A=ScalarParameter() {defaultValue=-6, group=CrystalSuperleveling, parameterName=BumpHeightA}, B=Subtract(A=@n21.R) {constB=0.5})
  n23 = TextureSampleParameter2D(UVs=StaticSwitchParameter(True=BumpOffset(Coordinate=@n20, Height=@n22), False=@n20) {defaultValue=true, group=CrystalSuperleveling, parameterName=UseBumpOffsetForNoiseTextures}) {
   group=CrystalSuperleveling
   parameterName=CrystalNoise
   samplerType=SAMPLERTYPE_Masks
   texture=/Game/Global/Textures/Default/Noise/NoisePattern
  }
  n24 = Panner(
   Coordinate=Divide(A=@n19, B=ScalarParameter() {defaultValue=110, group=CrystalSuperleveling, parameterName=CrystalTile_2})
   Speed=AppendVector(
    A=ScalarParameter() {group=CrystalSuperleveling, parameterName=CrystalPanSpeedU_2}
    B=ScalarParameter() {group=CrystalSuperleveling, parameterName=CrystalPanSpeedV_2}
   )
  ) {bFractionalPart=true}
  n25 = Multiply(A=ScalarParameter() {defaultValue=-3.5, group=CrystalSuperleveling, parameterName=BumpHeightB}, B=Subtract(A=@n21.G) {constB=0.5})
  n26 = TextureSampleParameter2D(UVs=StaticSwitchParameter(True=BumpOffset(Coordinate=@n24, Height=@n25), False=@n24) {defaultValue=true, group=CrystalSuperleveling, parameterName=UseBumpOffsetForNoiseTextures}) {
   group=CrystalSuperleveling
   parameterName=CrystalNoise
   samplerType=SAMPLERTYPE_Masks
   texture=&NoisePattern
  }
  n27 = Multiply(A=@n23.R, B=@n26.B)
  n28 = ScalarParameter() {defaultValue=3, group=CrystalSuperleveling, parameterName=SuperCrystalExponent}
  n29 = Add(A=@n23.A, B=@n26.G)
  n30 = ScalarParameter() {defaultValue=2, group=CrystalSuperleveling, parameterName=NoiseExponent}
  n31 = MaterialFunctionCall!(Time, Divisor=ScalarParameter() {defaultValue=10, group=CrystalSuperleveling, parameterName=ShimmerSpeedDivisor}, Mod) {materialFunction=/Game/Effects/Fort_Effects/Materials/_Functions/MF_ModuloTime}
  n32 = Add(A=@n21.RGB, B=Multiply(A=Sine(None=@n31.Unclamped), B=ScalarParameter() {group=CrystalSuperleveling, parameterName=ShimmerScale}))
  n33 = ReflectionVectorWS(CustomWorldNormal=Multiply(A=ConstantBiasScale(None=@n32) {bias=-0.5, scale=2}, B=Constant3Vector() {constant=<rgba 1 1 0 0>}))
  n34 = DotProduct(A=Normalize(VectorInput=Abs(None=@n33)), B=Constant3Vector() {constant=<rgba 0.5 0.5 0 0>})
  n35 = ScalarParameter() {defaultValue=18, group=CrystalSuperleveling, parameterName=ReflectionExponent}
  n36 = Add(
   A=Multiply(
    A=Power(Base=@n34, Exp=@n35) {constExponent=15}
    B=ScalarParameter() {defaultValue=200, group=CrystalSuperleveling, parameterName=ReflectionScale}
   )
   B=ScalarParameter() {defaultValue=0.2, group=CrystalSuperleveling, parameterName=ReflectionBias}
  )
  n37 = CurveAtlasRowParameter(CurveTime=Multiply(A=Power(Base=@n29, Exp=@n30), B=@n36)) {
   atlas=/Game/Characters/Player/Female/Medium/Bodies/F_MED_Shapeshifter/Curves/CA_Crystal
   curve=/Game/Characters/Player/Female/Medium/Bodies/F_MED_Shapeshifter/Curves/C_CrystalColor_Blue
   group=CrystalSuperleveling
   parameterName=CrystalColorAtlas
  }
  n38 = Multiply(
   A=Multiply(A=Power(Base=@n27, Exp=@n28) {constExponent=3}, B=@n37)
   B=VectorParameter() {group=CrystalSuperleveling, parameterName=SuperCrystalColor}.RGB
  )
  n39 = VectorParameter() {group=CrystalSuperleveling, parameterName=InnerCloudColorA, defaultValue=<rgba 0.1157 0.1887 0.4236 0>}
  n40 = VectorParameter() {group=CrystalSuperleveling, parameterName=InnerCloudColorB, defaultValue=<rgba 0.0257 0.0917 0.1042 0>}
  n41 = VectorParameter() {group=CrystalSuperleveling, parameterName=InnerCloudColorC, defaultValue=<rgba 0.1313 0.1081 0.1319 0>}
  n42 = Panner(
   Coordinate=Divide(A=@n19, B=ScalarParameter() {defaultValue=150, group=CrystalSuperleveling, parameterName=CloudTile})
   Speed=AppendVector(
    A=ScalarParameter() {group=CrystalSuperleveling, parameterName=CloudPanSpeedU}
    B=ScalarParameter() {group=CrystalSuperleveling, parameterName=CloudPanSpeedV}
   )
  ) {bFractionalPart=true}
  n43 = ScalarParameter() {defaultValue=-3, group=CrystalSuperleveling, parameterName=CloudHeight}
  n44 = TextureSampleParameter2D(UVs=StaticSwitchParameter(True=BumpOffset(Coordinate=@n42, Height=@n43), False=@n42) {defaultValue=true, group=CrystalSuperleveling, parameterName=UseBumpOffsetForNoiseTextures}) {
   group=CrystalSuperleveling
   parameterName=CrystalNoise
   samplerType=SAMPLERTYPE_Masks
   texture=&NoisePattern
  }
  n45 = TextureSampleParameter2D() {
   group=CrystalSuperleveling
   parameterName=CrystalMap
   samplerType=SAMPLERTYPE_LinearColor
   texture=&T_Shapeshifter_Crystal_Body
  }
  n46 = Add(
   A=Multiply(A=@n37, B=@n36)
   B=LinearInterpolate(
    A=MaterialFunctionCall(A=@n39.RGB, B=@n40.RGB, C=@n41.RGB, Alpha=@n44.G) {materialFunction=&Lerp_3Color}.Result
    B=VectorParameter() {group=CrystalSuperleveling, parameterName=AOColor, defaultValue=<rgba 0.1517 0.4438 0.9497 0>}.RGB
    Alpha=@n45.A
   )
  )
  n47 = Multiply(A=@n14.R, B=Multiply(A=Add(A=@n12.Fresnel, B=@n17), B=Add(A=@n38, B=@n46)))
  n48 = Add(A=Add(A=Saturate(None=@n5), B=LinearInterpolate(A=@n6.RGB, B=@n7.RGB, Alpha=@n13)), B=@n47)
  n49 = ChannelMaskParameter(None=TextureSampleParameter2D() {group=CrystalSuperleveling, parameterName=ModulateColorMaskTexture, texture=&T_WhitesRGB}.RGBA) {group=CrystalSuperleveling, parameterName=ModulateColorMask}
  n50 = ScalarParameter() {defaultValue=0.8, group=CrystalSuperleveling, parameterName=SpecularCrystal}
  n51 = Divide(A=@n19, B=ScalarParameter() {defaultValue=25, group=CrystalSuperleveling, parameterName=DetailNormalTileScale})
  n52 = TextureSampleParameter2D(UVs=@n51) {
   group=CrystalSuperleveling
   parameterName=CrystalMap
   samplerType=SAMPLERTYPE_LinearColor
   texture=&T_Shapeshifter_Crystal_Body
  }
  n53 = Multiply(A=ScalarParameter() {defaultValue=0.9, group=CrystalSuperleveling, parameterName=RoughnessCrystal}, B=@n52.B)
  n54 = TextureSampleParameter2D() {
   group=CrystalSuperleveling
   parameterName=Normals
   samplerType=SAMPLERTYPE_Normal
   texture=&T_F_MED_Shapeshifter_Body_N
  }
  n55 = TextureSampleParameter2D(UVs=Divide(A=@n19, B=ScalarParameter() {defaultValue=20, group=CrystalSuperleveling, parameterName=NormalTileScale})) {
   group=CrystalSuperleveling
   parameterName=CrystalNormal
   samplerType=SAMPLERTYPE_Normal
   texture=/Game/Weapons/WeaponSkins/Wraps/Textures/AlienRock/T_Wrap_AlienRock_n
  }
  n56 = Multiply(A=@n55.RGB, B=VectorParameter() {group=CrystalSuperleveling, parameterName=NormalInnerScale, defaultValue=<rgba 0.5 0.5 1 0>}.RGB)
  n57 = Normalize(VectorInput=Add(A=Subtract(A=@n54.RGB, B=@n56), B=@n56))
  n58 = TextureSampleParameter2D(UVs=@n51) {
   group=CrystalSuperleveling
   parameterName=DetailNormal
   samplerType=SAMPLERTYPE_Normal
   texture=/Game/Characters/Player/Female/Medium/Bodies/F_MED_Shapeshifter/Textures/T_Crystal_Scratches_Normal
  }
  n59 = SetMaterialAttributes!(
   MaterialAttributes
   Base Color=StaticSwitchParameter(True=Multiply(A=@n48, B=@n49.R), False=@n48) {parameterName=ModulateColor}
   Metallic=ScalarParameter() {defaultValue=0.8, group=CrystalSuperleveling, parameterName=MetallicCrystal}
   Specular=StaticSwitchParameter(True=Multiply(A=@n50, B=@n49.R), False=@n50) {parameterName=ModulateColor}
   Roughness=StaticSwitchParameter(True=LinearInterpolate(B=@n53, Alpha=@n49.R) {constA=10}, False=@n53) {parameterName=ModulateColor}
   Emissive Color=StaticSwitchParameter(True=Multiply(A=@n49.R, B=@n47), False=@n47) {parameterName=ModulateColor}
   Normal=StaticSwitchParameter(
    True=MaterialFunctionCall(BaseNormal=@n54.RGB, AdditionalNormal=@n57) {materialFunction=&BlendAngleCorrectedNormals}.Result
    False=MaterialFunctionCall(BaseNormal=@n56, AdditionalNormal=@n58.RGB) {materialFunction=&BlendAngleCorrectedNormals}.Result
   ) {group=CrystalSuperleveling, parameterName=UseCharacterNormalMapAsBase}
  ) {attributeSetTypes [
    #1
    #2
    #3
    #4
    #5
    #6
   ]}
  MaterialFunctionCall(AO=@n1.R, Distance Blend=@n2.Distance Blend, Fresnel=Constant(), Material=@n59) {materialFunction=&BaseFN_CharacterFunction_PostFX}
 ]
}
