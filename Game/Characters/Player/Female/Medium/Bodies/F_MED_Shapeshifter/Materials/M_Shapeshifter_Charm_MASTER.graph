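# Master character material for the Shapeshifter charm, with a Reveal parameter that expands or collapses the mesh.
# Shared character PreFX feeds a thin-film effect masked by an FX texture channel, then panning emissive and PostFX.
# The FX green channel and Reveal drive both the emissive sphere mask and a vertex collapse.
# World position offset pulls vertices toward the object pivot, shifted by offset and velocity parameters, as Reveal drops.
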Material /Game/Characters/Player/Female/Medium/Bodies/F_MED_Shapeshifter/Materials/M_Shapeshifter_Charm_MASTER
graph {
 properties {
  bUseMaterialAttributes=true
  bUsedWithInstancedStaticMeshes=true
  bUsedWithMeshDeformer=true
  bUsedWithMorphTargets=true
  bUsedWithSkeletalMesh=true
  ditherOpacityMask=true
  pixelDepthOffsetMode=PDOM_Legacy
  usedShadingModels=Default Lit
 }
 nodes [
  ChannelMaskParameter() {parameterName=ThinFilm Channel}
  Constant()
  n1 = MaterialFunctionCall() {materialFunction=&BaseFN_CharacterFunction_PreFX}
  n2 = ScalarParameter() {defaultValue=1, parameterName=Reveal}
  n3 = TextureSampleParameter2D() {parameterName=FX, samplerType=SAMPLERTYPE_Masks, texture=/Game/Global/Textures/Default/Blanks/AllRed}
  n4 = OneMinus(None=@n3.G)
  n5 = ScalarParameter() {defaultValue=1, parameterName=Thickness}
  n6 = Constant()
  n7 = ChannelMaskParameter(None=@n3.RGBA) {parameterName=ThinFilm Channel}
  n8 = MaterialFunctionCall(
   AO=@n1.AO
   Distance Blend=@n1.AO
   Fresnel=@n1.AO
   Gmap=@n1.AO
   Pre Skinned Local Position=@n1.AO
   Material=MaterialFunctionCall!(
    Alpha=SphereMask(A=@n2, B=@n4, Radius=@n5, Hardness=@n6)
    isVehicle
    Gmap=@n1.Gmap
    UVOffset
    Local Normal
    Local Position=@n1.Gmap
    EmissiveColorOverride
    Material=MaterialFunctionCall!(Alpha=@n7.R, Material Attributes=@n1.Result) {materialFunction=/Game/Weapons/WeaponSkins/MaterialFunction/ThinFilmFunction}.Result
   ) {materialFunction=/Game/Weapons/WeaponSkins/MaterialFunction/PanningEmissive}.Result
  ) {materialFunction=&BaseFN_CharacterFunction_PostFX}
  n9 = GetMaterialAttributes(None=@n8.Result) {attributeGetTypes [#1]}
  n10 = ObjectPositionWS()
  n11 = VectorParameter() {parameterName=ObjectPivotOffset, defaultValue=<rgba 0 0 0 1>}
  n12 = Subtract(
   A=WorldPosition().XYZ
   B=Add(A=Add(A=@n10, B=@n11.RGB), B=VectorParameter() {parameterName=ObjectPivotVelocity, defaultValue=<rgba 0 0 0 1>}.RGB)
  )
  n13 = Add(A=Subtract(A=@n3.G), B=LinearInterpolate(Alpha=@n2) {constA=1.8, constB=0})
  SetMaterialAttributes!(
   MaterialAttributes=@n9.MaterialAttributes
   World Position Offset=LinearInterpolate(A=@n9.World Position Offset, B=Multiply(A=@n12) {constB=-1}, Alpha=Saturate(None=@n13))
  ) {attributeSetTypes [#1]}
 ]
}
