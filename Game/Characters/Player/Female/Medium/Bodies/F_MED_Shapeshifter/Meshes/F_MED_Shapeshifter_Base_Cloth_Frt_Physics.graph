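# Physics asset for the front cloth of the female medium Shapeshifter player body, defining capsule bodies and joints for a hanging cloth bone chain.
# Pelvis, thigh and calf capsules form a lower-body proxy around the cloth.
# A driver bone joins the pelvis with free swing and twist and leads a five-bone cloth chain.
# Chain swing limits widen over the first three joints, while twist stays tightly limited throughout.
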PhysicsAsset /Game/Characters/Player/Female/Medium/Bodies/F_MED_Shapeshifter/Meshes/F_MED_Shapeshifter_Base_Cloth_Frt_Physics
bodies {
 calf_l [{shapeType=Capsule, length=28.8881, radius=9, center=<-28.6828 -0.4834 -0.7508>, rotation=<rot 0 76.037 -89.3289>}]
 calf_r [{shapeType=Capsule, length=28.8888, radius=9, center=<28.684 0.4854 0.7526>, rotation=<rot 0 76.0315 -89.3201>}]
 dyn_cloth_front_01 [{shapeType=Capsule, length=7, radius=8.25, center=<1 8 0>, rotation=<rot 90 0 0>}]
 dyn_cloth_front_02 [{shapeType=Capsule, length=7, radius=8.25, center=<1.2 8 0>, rotation=<rot 90 0 0>}]
 dyn_cloth_front_03 [{shapeType=Capsule, length=7, radius=8.25, center=<1.4 8 0>, rotation=<rot 90 0 0>}]
 dyn_cloth_front_04 [{shapeType=Capsule, length=7, radius=8.25, center=<1.6 8 0>, rotation=<rot 90 0 0>}]
 dyn_cloth_front_05 [{shapeType=Capsule, length=7, radius=8.25, center=<1.8 8 0>, rotation=<rot 90 0 0>}]
 dyn_driver_cloth_front [{shapeType=Capsule, length=8, radius=7.25, rotation=<rot 90 0 0>}]
 pelvis [{shapeType=Capsule, length=7.75, radius=10.75, center=<-3.7758 -1.9704 0>}]
 thigh_l [{shapeType=Capsule, length=40.8145, radius=9.75, center=<-23.989 -1.7186 -0.0002>, rotation=<rot 90 0 -1.0629>}]
 thigh_r [{shapeType=Capsule, length=41.0645, radius=9.5, center=<23.989 1.7186 -0.0002>, rotation=<rot 90 0 -1.0629>}]
}
constraints [
 {bone1Name=thigh_l, bone2Name=pelvis, swing1LimitDegrees=45, swing2LimitDegrees=45, twistLimitDegrees=45}
 {bone1Name=calf_l, bone2Name=thigh_l, swing1LimitDegrees=45, swing2LimitDegrees=45, twistLimitDegrees=45}
 {bone1Name=thigh_r, bone2Name=pelvis, swing1LimitDegrees=45, swing2LimitDegrees=45, twistLimitDegrees=45}
 {bone1Name=calf_r, bone2Name=thigh_r, swing1LimitDegrees=45, swing2LimitDegrees=45, twistLimitDegrees=45}
 {
  bone1Name=dyn_driver_cloth_front
  bone2Name=pelvis
  swing1LimitDegrees=45
  swing1Motion=Free
  swing2LimitDegrees=45
  swing2Motion=Free
  twistLimitDegrees=45
  twistMotion=Free
 }
 {bone1Name=dyn_cloth_front_01, bone2Name=dyn_driver_cloth_front, swing1LimitDegrees=25, swing2LimitDegrees=25, twistLimitDegrees=5}
 {bone1Name=dyn_cloth_front_02, bone2Name=dyn_cloth_front_01, swing1LimitDegrees=35, swing2LimitDegrees=35, twistLimitDegrees=5}
 {bone1Name=dyn_cloth_front_03, bone2Name=dyn_cloth_front_02, swing1LimitDegrees=45, swing2LimitDegrees=45, twistLimitDegrees=5}
 {bone1Name=dyn_cloth_front_04, bone2Name=dyn_cloth_front_03, swing1LimitDegrees=45, swing2LimitDegrees=45, twistLimitDegrees=5}
 {bone1Name=dyn_cloth_front_05, bone2Name=dyn_cloth_front_04, swing1LimitDegrees=45, swing2LimitDegrees=45, twistLimitDegrees=5}
]
