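# Physics asset for the medium female Shapeshifter's light armor mesh, set up for secondary motion on a hanging pouch.
# A tiny driver capsule anchors the pouch capsule through a single angular-limited constraint that allows moderate swing and little twist.
# An unconstrained right thigh capsule is the only other body, apparently a collider to keep the pouch out of the leg.
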PhysicsAsset /Game/Characters/Player/Female/Medium/Bodies/F_MED_Shapeshifter/Meshes/F_MED_Shapeshifter_Light_Armor_Physics
bodies {
 dyn_driver_light_pouch_01 [{shapeType=Capsule, length=0.505, radius=0.505, rotation=<rot 0 -90 0>}]
 dyn_light_pouch_01 [{shapeType=Capsule, length=8, radius=3.25, center=<6.5146 -2.5 0>, rotation=<rot 90 0 0>}]
 thigh_r [{shapeType=Capsule, length=20.75, radius=8.75, center=<13.0117 0.0001 -0.7745>, rotation=<rot 87.8497 179.9998 179.9998>}]
}
constraints [{bone1Name=dyn_light_pouch_01, bone2Name=dyn_driver_light_pouch_01, swing1LimitDegrees=20, swing2LimitDegrees=20, twistLimitDegrees=5}]
