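# Post-process style animation blueprint for the female Shapeshifter player body, adding armor and cloth secondary motion and swapping in a tail for swimming.
# The AnimGraph copies the parent mesh pose, then layers Control Rigs, corrective bone transforms and RigidBody physics.
# Updates derive corrective alphas from limb rotations, gravity overrides and front-end state.
# Meta tags select armor or disable the tail; swimming, dive-jumping or preview reactions show it, hiding footwear and material elements.
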AnimBlueprint /Game/Characters/Player/Female/Medium/Bodies/F_MED_Shapeshifter/Meshes/F_MED_Shapeshifter_AnimBP
target_skeleton=&/F_MED_Shapeshifter_Skeleton
graphs [
 AnimBP {nodes [
   Inputs!() -> (NewParam:Exec)
   n1 = Get PawnOwner() -> (PawnOwner:Player Pawn Athena Object Reference)
   > n2 = Cast To Fortnite_M_Avg_Player_AnimBlueprint!(execute, Object=GetAnimInstance(self=Get Mesh(self=@n1))) -> (then, CastFailed, AsFortnite M Avg Player Anim Blueprint:Fortnite M Avg Player Anim Blueprint Object Reference) {is=title K2Node_DynamicCast}
   > Outputs!(Exec:Exec, CastFailed:Exec=@n2.CastFailed, "AnimBP (DO NOT CACHE)":Fortnite M Avg Player Anim Blueprint Object Reference=@n2)
  ]}
 AnimGraph {nodes [
   n1 = Get GravityOverride_b() -> (GravityOverride_b:Vector)
   n2 = Look At - Bone: aimer_shldr_pad_r(ComponentPose=Local To Component(LocalPose=Use cached pose 'Pose'()), Alpha=1) {
    alphaBoolBlend {blendOption=Linear}
    boneToModify {boneName=aimer_shldr_pad_r}
    lookAtLocation=<-25.5636 12.1774 -9.2139>
    lookAtTarget {boneReference {boneName=upperarm_r}}
    lookAt_Axis {axis=<1 0>}
   }
   n3 = Control Rig!(Source=Use cached pose 'Pose'(), Alpha, ArmorOffset_X:double=Get armor_offsetX() -> (armor_offsetX:double)) -> (Pose) {
    alphaBoolBlend {blendOption=Linear}
    controlRigAssetReference {blueprintRigClass=&/F_MED_Shapeshifter_Armor_CtrlRig.F_MED_Shapeshifter_Armor_CtrlRig_C}
   }
   n4 = Control Rig!(
    Source=Blend Poses by bool(BlendPose_0=Component To Local(ComponentPose=@n2), BlendPose_1=@n3, bActiveValue=Get bIsPlayingEmote()) {blendType=Linear}
    Alpha=Get shldr_r_roll_alpha() -> (shldr_r_roll_alpha:double)
   ) -> (Pose) {
    alphaBoolBlend {blendOption=Linear}
    alphaScaleBiasClamp {bInterpResult=true, interpSpeedDecreasing=5, interpSpeedIncreasing=7}
    controlRigAssetReference {blueprintRigClass=&/F_MED_Shapeshifter_CtrlRig.F_MED_Shapeshifter_CtrlRig_C}
   }
   n5 = "Constraint - shldr_eagle_l (PPP)"!(
    ComponentPose=Local To Component(LocalPose=@n4)
    Alpha=1
    ConstraintWeights_0=0.6298
    ConstraintWeights_1:float=0.0134
    ConstraintWeights_2:float=0.3568
   ) -> (Pose) {
    alphaBoolBlend {blendOption=Linear}
    boneToModify {boneName=shldr_eagle_l}
    constraintSetup [{targetBone {boneName=upperarm_twist_01_l}}, {targetBone {boneName=spine_05}}, {targetBone {boneName=clavicle_l}}]
    constraintWeights [1,1,1]
   }
   n6 = "Transform (Modify) Bone - Bone: shldr_pad_bck_01"(ComponentPose=@n5, Alpha=Get light_armor_alpha() -> (light_armor_alpha:double), Translation="0,0,0", Rotation="-10,-5,0", Scale="1,1,1") {
    rotationMode=BMM_Additive
    rotationSpace=BCS_BoneSpace
    alphaBoolBlend {blendOption=Linear}
    boneToModify {boneName=shldr_pad_bck_01}
    rotation {pitch=-10, yaw=-5}
   }
   n7 = "Transform (Modify) Bone - Bone: dyn_glute_l"(ComponentPose=@n6, Alpha=Get thigh_rot_l(), Translation="-4,0.875,-0.9", Rotation="0.000,0.000,-55", Scale="1.3,1.3,1.3") {
    rotationMode=BMM_Additive
    rotationSpace=BCS_BoneSpace
    scaleMode=BMM_Additive
    scaleSpace=BCS_BoneSpace
    translationMode=BMM_Additive
    translationSpace=BCS_BoneSpace
    alphaBoolBlend {blendOption=Linear}
    boneToModify {boneName=dyn_glute_l}
    rotation {roll=-55}
    scale=<1.3 1.3 1.3>
    translation=<-4 0.875 -0.9>
   }
   n8 = "Transform (Modify) Bone - Bone: dyn_glute_r"(ComponentPose=@n7, Alpha=Get thigh_rot_r(), Translation="-4,0,0", Rotation="0,0,-55", Scale="1.3,1.3,1.3") {
    rotationMode=BMM_Additive
    rotationSpace=BCS_BoneSpace
    scaleMode=BMM_Additive
    scaleSpace=BCS_BoneSpace
    translationMode=BMM_Additive
    translationSpace=BCS_BoneSpace
    alphaBoolBlend {blendOption=Linear}
    boneToModify {boneName=dyn_glute_r}
    rotation {roll=-55}
    scale=<1.3 1.3 1.3>
    translation {x=-4}
   }
   n9 = Get knee_rot_alpha_l() -> (knee_rot_alpha_l:double)
   n10 = "Transform (Modify) Bone - Bone: dyn_kneepad_l"(ComponentPose=@n8, Alpha=@n9, Translation="0,-1,0", Rotation="0,-17,0", Scale="1,1,1") {
    rotationMode=BMM_Additive
    rotationSpace=BCS_BoneSpace
    translationMode=BMM_Additive
    translationSpace=BCS_BoneSpace
    alphaBoolBlend {blendOption=Linear}
    alphaScaleBiasClamp {bInterpResult=true, interpSpeedDecreasing=20, interpSpeedIncreasing=20}
    boneToModify {boneName=dyn_kneepad_l}
    rotation {yaw=-17}
    translation {y=-1}
   }
   n11 = "Transform (Modify) Bone - Bone: dyn_thigh_link"(ComponentPose=@n10, Alpha=@n9, Translation="0,0,0", Rotation="15,0,0", Scale="0.8,0.8,0.8") {
    rotationMode=BMM_Additive
    rotationSpace=BCS_BoneSpace
    scaleMode=BMM_Additive
    scaleSpace=BCS_BoneSpace
    translationMode=BMM_Additive
    translationSpace=BCS_BoneSpace
    alphaBoolBlend {blendOption=Linear}
    alphaScaleBiasClamp {bInterpResult=true, interpSpeedDecreasing=20, interpSpeedIncreasing=20}
    boneToModify {boneName=dyn_thigh_link}
    rotation {pitch=15}
    scale=<0.8 0.8 0.8>
   }
   n12 = "Transform (Modify) Bone - Bone: dyn_kneepad_r"(ComponentPose=@n11, Alpha=Get knee_rot_alpha_r() -> (knee_rot_alpha_r:double), Translation="0,-1,0", Rotation="0,17,0", Scale="1,1,1") {
    rotationMode=BMM_Additive
    rotationSpace=BCS_BoneSpace
    translationMode=BMM_Additive
    translationSpace=BCS_BoneSpace
    alphaBoolBlend {blendOption=Linear}
    alphaScaleBiasClamp {bInterpResult=true, interpSpeedDecreasing=20, interpSpeedIncreasing=20}
    boneToModify {boneName=dyn_kneepad_r}
    rotation {yaw=17}
    translation {y=-1}
   }
   n13 = "Transform (Modify) Bone - Bone: shldr_pad_r"(
    ComponentPose=Copy Bone - Source Bone: hand_r - Target Bone: dyn_wrist_horn_01!(
     ComponentPose:Component Space Pose Link Structure=@n12
     Alpha:float=0.75
     bCopyTranslation:bool=False
     bCopyRotation:bool=true
     bCopyScale:bool=False
    ) -> (Pose:Component Space Pose Link Structure) {is=title AnimGraphNode_CopyBone, alphaBoolBlend {blendOption=Linear}, sourceBone {boneName=hand_r}, targetBone {boneName=dyn_wrist_horn_01}}
    Alpha=Get skydiving_arm_up() -> (skydiving_arm_up:double)
    Translation="0,0,0"
    Rotation="35,0,0"
    Scale="1,1,1"
   ) {
    rotationMode=BMM_Additive
    rotationSpace=BCS_BoneSpace
    scaleSpace=BCS_BoneSpace
    translationMode=BMM_Additive
    translationSpace=BCS_BoneSpace
    alphaBoolBlend {blendOption=Linear}
    alphaScaleBiasClamp {bInterpResult=true, interpSpeedDecreasing=15, interpSpeedIncreasing=15}
    boneToModify {boneName=shldr_pad_r}
    rotation {pitch=35}
   }
   n14 = "Transform (Modify) Bone - Bone: shldr_pad_r"(ComponentPose=@n13, Alpha=Get shldr_r_roll_alpha(), Translation="1,1,-0.2", Rotation="-10,-15,15", Scale="1,1,1") {
    rotationMode=BMM_Additive
    rotationSpace=BCS_BoneSpace
    scaleSpace=BCS_BoneSpace
    translationMode=BMM_Additive
    translationSpace=BCS_BoneSpace
    alphaBoolBlend {blendOption=Linear}
    alphaScaleBiasClamp {interpSpeedDecreasing=60, interpSpeedIncreasing=60}
    boneToModify {boneName=shldr_pad_r}
    rotation=<rot -10 -15 15>
    translation=<1 1 -0.2>
   }
   n15 = "Transform (Modify) Bone - Bone: shldr_pad_r"(ComponentPose=@n14, Alpha=Get shoulder_pad_float() -> (shoulder_pad_float:double), Translation="0,0,0", Rotation="20,0,0", Scale="1,1,1") {
    rotationMode=BMM_Additive
    rotationSpace=BCS_BoneSpace
    scaleSpace=BCS_BoneSpace
    translationMode=BMM_Additive
    translationSpace=BCS_BoneSpace
    alphaBoolBlend {blendOption=Linear}
    alphaScaleBiasClamp {bInterpResult=true, interpSpeedDecreasing=15, interpSpeedIncreasing=15}
    boneToModify {boneName=shldr_pad_r}
    rotation {pitch=20}
   }
   n16 = Flap Node - Bone: dyn_driver_pouch_01!(ComponentPose:Component Space Pose Link Structure=@n15, Alpha:float=1) -> (Pose:Component Space Pose Link Structure) {
    is=title FortAnimGraphNode_Flap
    bUseClamp=true
    targetClampMax=1
    targetClampMin=-0.55
    alphaBoolBlend {blendOption=Linear}
    alphaScaleBiasClamp {bInterpResult=true, interpSpeedDecreasing=20, interpSpeedIncreasing=20}
    sourceDrives [{
      bUseRange=false
      drivingCurve=null
      multiplier=-0.5
      rangeMax=1
      rangeMin=-1
      remappedMax=1
      remappedMin=0
      sourceComponent=RotationZ
      useQuaternion=true
      sourceBone {boneName=thigh_r}
     }]
    targetBone {boneName=dyn_driver_pouch_01}
   }
   n17 = RigidBody(ComponentPose=@n16) {
    bOverrideWorldGravity=true
    overridePhysicsAsset=&/F_MED_Shapeshifter_Base_Pouch_Physics
    simulationSpace=BaseBoneSpace
    alphaBoolBlend {blendOption=Linear}
    baseBoneRef {boneName=root}
    componentAppliedLinearAccClamp=<1100 1100 1100>
    componentLinearAccScale=<0.3 0.3 0.4>
    componentLinearVelScale=<0.3 0.3 0.4>
    overrideWorldGravity {z=-980}
   }
   n18 = RigidBody(ComponentPose=@n17) {
    bOverrideWorldGravity=true
    overridePhysicsAsset=&/F_MED_Shapeshifter_Light_Armor_Physics
    simulationSpace=BaseBoneSpace
    alphaBoolBlend {blendOption=Linear}
    baseBoneRef {boneName=root}
    componentAppliedLinearAccClamp=<1100 1100 1100>
    componentLinearAccScale=<0.3 0.3 0.4>
    componentLinearVelScale=<0.3 0.3 0.4>
    overrideWorldGravity {z=-980}
   }
   n19 = RigidBody!(ComponentPose=@n18, Alpha=Get cloth_alpha() -> (cloth_alpha:double), OverrideWorldGravity:Vector=@n1, ExternalForce) -> (Pose) {
    bOverrideWorldGravity=true
    overridePhysicsAsset=&/F_MED_Shapeshifter_Base_Cloth_Frt_Physics
    simulationSpace=BaseBoneSpace
    alphaBoolBlend {blendOption=Linear}
    baseBoneRef {boneName=root}
    componentAppliedLinearAccClamp=<1100 1100 1100>
    componentLinearAccScale=<0.3 0.3 0.4>
    componentLinearVelScale=<0.3 0.3 0.4>
    overrideWorldGravity {z=-980}
   }
   n20 = RigidBody(ComponentPose=@n19, Alpha=Get cloth_light_alpha() -> (cloth_light_alpha:double), OverrideWorldGravity=@n1) {
    bOverrideWorldGravity=true
    overridePhysicsAsset=&/F_MED_Shapeshifter_Light_Cloth_Frt_Physics
    simulationSpace=BaseBoneSpace
    alphaBoolBlend {blendOption=Linear}
    baseBoneRef {boneName=root}
    componentAppliedLinearAccClamp=<1100 1100 1100>
    componentLinearAccScale=<0.3 0.3 0.4>
    componentLinearVelScale=<0.3 0.3 0.4>
    overrideWorldGravity {z=-980}
   }
   n21 = RigidBody(ComponentPose=@n20, Alpha=Get rear_cloth_alpha() -> (rear_cloth_alpha:double), OverrideWorldGravity=Get GravityOverride_b()) {
    bOverrideWorldGravity=true
    overridePhysicsAsset=&/F_MED_Shapeshifter_Base_Cloth_Bck_Physics
    simulationSpace=BaseBoneSpace
    alphaBoolBlend {blendOption=Linear}
    baseBoneRef {boneName=root}
    componentAppliedLinearAccClamp=<1100 1100 1100>
    componentLinearAccScale=<0.3 0.3 0.4>
    componentLinearVelScale=<0.3 0.3 0.4>
    overrideWorldGravity {z=-980}
   }
   n22 = Flap Node - Bone: dyn_buckle_01!(ComponentPose:Component Space Pose Link Structure=@n21, Alpha:float=Get buckle_alpha() -> (buckle_alpha:double)) -> (Pose:Component Space Pose Link Structure) {
    is=title FortAnimGraphNode_Flap
    bUseClamp=true
    targetClampMax=10
    targetComponent=RotationZ
    alphaBoolBlend {blendOption=Linear}
    sourceDrives [{
      bUseRange=false
      drivingCurve=null
      multiplier=-1
      rangeMax=1
      rangeMin=-1
      remappedMax=1
      remappedMin=0
      sourceComponent=RotationZ
      useQuaternion=true
      sourceBone {boneName=dyn_cloth_front_01}
     }]
    targetBone {boneName=dyn_buckle_01}
   }
   n23 = "Transform (Modify) Bone - Bone: thigh_twist_01_l"(
    ComponentPose=@n22
    Alpha=Get frontEnd_thigh_alpha() -> (frontEnd_thigh_alpha:double)
    Translation="0,0,0"
    Rotation="-3,0.000,13"
    Scale="1,1,1"
   ) {
    rotationMode=BMM_Additive
    rotationSpace=BCS_BoneSpace
    alphaBoolBlend {blendOption=Linear}
    boneToModify {boneName=thigh_twist_01_l}
    rotation {pitch=-3, roll=13}
   }
   n24 = Get frontEnd_cloth_alpha() -> (frontEnd_cloth_alpha:double)
   n25 = "Transform (Modify) Bone - Bone: dyn_driver_cloth_front"(ComponentPose=@n23, Alpha=@n24, Translation="0,-0.85,0", Rotation="0,-6,0", Scale="1,1,1") {
    rotationMode=BMM_Additive
    rotationSpace=BCS_BoneSpace
    translationMode=BMM_Additive
    translationSpace=BCS_BoneSpace
    alphaBoolBlend {blendOption=Linear}
    boneToModify {boneName=dyn_driver_cloth_front}
    rotation {yaw=-6}
    translation {y=-0.85}
   }
   n26 = "Transform (Modify) Bone - Bone: dyn_driver_cloth_rear"(ComponentPose=@n25, Alpha=@n24, Translation="0,0,0", Rotation="0,5,0", Scale="1,1,1") {
    rotationMode=BMM_Additive
    rotationSpace=BCS_BoneSpace
    alphaBoolBlend {blendOption=Linear}
    boneToModify {boneName=dyn_driver_cloth_rear}
    rotation {yaw=5}
   }
   n27 = Look At - Bone: dyn_parent_cloth_front!(ComponentPose=@n26, bAlphaBoolEnabled:bool=True) -> (Pose) {
    alphaInputType=Bool
    alphaBoolBlend {blendOption=Linear}
    boneToModify {boneName=dyn_parent_cloth_front}
    lookAtLocation {x=50}
    lookAtTarget {boneReference {boneName=calf_r}}
    lookAt_Axis {axis=<1 -0.2>}
   }
   Output Pose(Result=Component To Local(ComponentPose=@n27)) {layerGroup=None}
   Save cached pose 'Pose'(Pose=Copy Pose From Mesh() {bUseAttachedParent=true}) {cacheName=Pose}
  ]}
 EventGraph {nodes [
   n1 = PrintString!(execute, InString=Huzzah, bPrintToScreen, bPrintToLog, TextColor, Duration, Key) -> (then)
   Delay(execute=[@n1.then, @n49.then], Duration=1.2)
   > Set bIsPreviewReacting_Swimming(bIsPreviewReacting_Swimming:bool=true) -> (Output_Get:bool)
   Event OnBeginSkydiving()
   > Set actively_skydiving(actively_skydiving:bool=true) -> (Output_Get:bool)
   Event BlueprintInitializeAnimation()
   > n2 = Cast To PlayerPawn_Athena(Object=TryGetPawnOwner())
   > Set PawnOwner(PawnOwner:Player Pawn Athena Object Reference=@n2) -> (Output_Get:Player Pawn Athena Object Reference)
   n3 = Event StopPreviewReactiveResponse!() -> (
    OutputDelegate
    then
    ChannelAndVariant:"Cosmetic Variant Info Structure (by ref)"
    Preview:"Fort Cosmetic Reactive Preview (F CRP) Object Reference"
   ) {is=title K2Node_Event}
   > Gate Is Cosmetic Variant to React Towards!(execute:Exec, CosmeticVariantInfo:"Cosmetic Variant Info Structure (by ref)"=@n3.ChannelAndVariant) -> (then:Exec) {is=title K2Node_MacroInstance}
   > Set bIsPreviewReacting_Swimming(bIsPreviewReacting_Swimming=false)
   Event OnEndSkydiving()
   > Set actively_skydiving(actively_skydiving=false)
   Event OnBeginCrouching()
   > Set crouch_alpha(crouch_alpha:double=1) -> (Output_Get:double)
   Event OnCharacterCustomizationLoaded()
   > n4 = Sequence!(execute) -> (then_0, then_1, then_2:Exec, then_3:Exec)
   n5 = Get PawnOwner()
   > Set BodyMesh(BodyMesh=GetSkeletalMeshForPartType(self=@n5))
   > n6 = Set TailMesh(TailMesh:Skeletal Mesh Component Object Reference=GetSkeletalMeshForPartType(self=@n5, PartType=MiscOrTail)) -> (Output_Get:Skeletal Mesh Component Object Reference)
   > SetVisibility!(execute, self:Scene Component Object Reference=@n6, bNewVisibility:bool=false, bPropagateToChildren:bool=false) -> (then)
   > n7 = Get Dynamic Material Instance(SkeletalMeshComponent=Get BodyMesh() -> (BodyMesh:Skeletal Mesh Component Object Reference))
   > Set Mat_Instance(Mat_Instance:Material Instance Dynamic Object Reference=@n7) -> (Output_Get:Material Instance Dynamic Object Reference)
   > Set CosmeticShoesMesh(CosmeticShoesMesh:Skeletal Mesh Component Object Reference=GetSkeletalMeshForPartType(self=@n5, PartType=ExtraPart)) -> (Output_Get:Skeletal Mesh Component Object Reference)
   n8 = GetFortAbilitySystemComponent!(Actor:Actor Object Reference=@n5) -> (ReturnValue:Fort Ability System Component Object Reference)
   > Branch(Condition=HasTag!(
     TagContainer:"Gameplay Tag Container Structure (by ref)"=Get Owned Gameplay Tags!(self:Gameplay Tag Asset Interface Interface=@n8) -> (ReturnValue:Gameplay Tag Container Structure) {is=title K2Node_CallFunction}
     Tag:Gameplay Tag Structure=(TagName="Ability.Primal.Item.JumpBoots.Equipped")
     bExactMatch:bool=true
    ) -> (ReturnValue:bool))
   > Set JumpBootsMesh(JumpBootsMesh:Skeletal Mesh Component Object Reference=GetSkeletalMeshForPartType(self=@n5, PartType=Gameplay)) -> (Output_Get:Skeletal Mesh Component Object Reference)
   n9 = Branch(
    execute=@n4.then_1
    Condition=CharacterItemDefinitionHasAnyMetaTag(Tags=Make Array("[0]"=(TagName="Cosmetics.MetaTags.ShapeShifter.TailOff")))
   )
   > Set disable_tail(disable_tail:bool=true) -> (Output_Get:bool)
   Set disable_tail(execute=@n9.else, disable_tail=false)
   n10 = Branch(
    execute=@n4.then_2
    Condition=CharacterItemDefinitionHasAnyMetaTag(Tags=Make Array("[0]"=(TagName="Cosmetics.MetaTags.ShapeShifter.LightArmor")))
   )
   > Set armor_offsetX(armor_offsetX:double=3.25) -> (Output_Get:double)
   > Set base_not_active(base_not_active:bool=true) -> (Output_Get:bool)
   > Set light_armor_alpha(light_armor_alpha:double=1) -> (Output_Get:double)
   n11 = Branch(
    execute=@n10.else
    Condition=CharacterItemDefinitionHasAnyMetaTag(Tags=Make Array("[0]"=(TagName="Cosmetics.MetaTags.ShapeShifter.HeavyArmor")))
   )
   > Set armor_offsetX(armor_offsetX=3.75)
   > Set base_not_active(base_not_active=false)
   > Set light_armor_alpha(light_armor_alpha=0)
   Set armor_offsetX(execute=@n11.else, armor_offsetX=SelectFloat(A=0, B=0.7, bPickA=Get emote_is_active() -> (emote_is_active:bool)))
   > Set base_not_active(base_not_active=false)
   > Set light_armor_alpha(light_armor_alpha=0)
   Event OnEndCrouching()
   > Set crouch_alpha(crouch_alpha=0)
   n12 = Event BlueprintUpdateAnimation()
   > Get!(execute:Exec) -> (then:Exec, else:Exec, BodyMesh:Skeletal Mesh Component Object Reference)
   > n13 = Sequence!(execute) -> (then_0, then_1, then_2, then_3, then_4:Exec, then_5:Exec, then_6:Exec)
   > n14 = Anim BP!(NewParam:Exec) -> (Exec:Exec, CastFailed:Exec, "AnimBP (DO NOT CACHE)":Fortnite M Avg Player Anim Blueprint Object Reference) {is=title K2Node_MacroInstance}
   n15 = Get BodyMesh()
   n16 = Get Dynamic Material Instance(execute=[@n14.CastFailed, @n29.else], SkeletalMeshComponent=@n15)
   > SetScalarParameterValue(self=@n16, ParameterName=Hide Element 01, Value=0)
   > SetScalarParameterValue(self=@n16, ParameterName=Hide Element 02, Value=0)
   n17 = Get TailMesh() -> (TailMesh:Skeletal Mesh Component Object Reference)
   > SetVisibility(self=@n17, bNewVisibility=false, bPropagateToChildren=false)
   n18 = Get CosmeticShoesMesh() -> (CosmeticShoesMesh:Skeletal Mesh Component Object Reference)
   > n19 = Is Valid(InputObject=@n18)
   > n20 = SetVisibility(self=@n18, bNewVisibility=true, bPropagateToChildren=false)
   n21 = Get JumpBootsMesh() -> (JumpBootsMesh:Skeletal Mesh Component Object Reference)
   SetVisibility(execute=[@n20.then, @n19.Is Not Valid], self=@n21, bNewVisibility=true, bPropagateToChildren=false)
   > Set changed_swimming(changed_swimming:bool=false) -> (Output_Get:bool)
   > Set buckle_alpha(buckle_alpha:double=1) -> (Output_Get:double)
   > Call Tail Off!(execute:Exec, self:F MED Shapeshifter Anim BP Object Reference) -> (then:Exec) {is=title K2Node_CallDelegate}
   > n22 = Branch(Condition=Get base_not_active() -> (base_not_active:bool))
   > Set cloth_alpha(cloth_alpha:double=0) -> (Output_Get:double)
   > Set cloth_light_alpha(cloth_light_alpha:double=1) -> (Output_Get:double)
   > Set rear_cloth_alpha(rear_cloth_alpha:double=1) -> (Output_Get:double)
   Set cloth_alpha(execute=@n22.else, cloth_alpha=1)
   > Set cloth_light_alpha(cloth_light_alpha=0)
   > Set rear_cloth_alpha(rear_cloth_alpha=1)
   n23 = Get bIsInVehicle!(self:Fort Player Anim Instance Object Reference=@n14) -> (bIsInVehicle:bool)
   n24 = Get disable_tail() -> (disable_tail:bool)
   n25 = Get bIsDiveJumping!(self:Fort Player Anim Instance Object Reference=@n14) -> (bIsDiveJumping:bool)
   n26 = Get bIsSurfaceSwimming()
   n27 = Get bIsSwimSprinting() -> (bIsSwimSprinting:bool)
   n28 = Get bIsPreviewReacting_Swimming() -> (bIsPreviewReacting_Swimming:bool)
   n29 = Branch(
    execute=@n14.Exec
    Condition=AND Boolean!(A=OR Boolean!(A=@n26, B=@n27, C:bool=@n25, D:bool=@n28) -> (ReturnValue), B=NOT Boolean(A=@n23), C:bool=NOT Boolean(A=@n24)) -> (ReturnValue)
   )
   > n30 = Get Dynamic Material Instance(SkeletalMeshComponent=@n15)
   > SetScalarParameterValue(self=@n30, ParameterName=Hide Element 01, Value=1)
   > SetScalarParameterValue(self=@n30, ParameterName=Hide Element 02, Value=1)
   > SetVisibility(self=@n17, bNewVisibility=true, bPropagateToChildren=false)
   > n31 = Is Valid(InputObject=@n18)
   > n32 = SetVisibility(self=@n18, bNewVisibility=false, bPropagateToChildren=false)
   SetVisibility(execute=[@n32.then, @n31.Is Not Valid], self=@n21, bNewVisibility=false, bPropagateToChildren=false)
   > Set changed_swimming(changed_swimming=false)
   > Set buckle_alpha(buckle_alpha=0)
   > Set cloth_alpha(cloth_alpha=0)
   > Set cloth_light_alpha(cloth_light_alpha=0)
   > Set rear_cloth_alpha(rear_cloth_alpha=0)
   > Call Tail On!(execute:Exec, self:F MED Shapeshifter Anim BP Object Reference) -> (then:Exec) {is=title K2Node_CallDelegate}
   n33 = Branch(execute=@n13.then_5, Condition=Get bIsInFrontEnd())
   > n34 = Branch(Condition=NOT Boolean(A=Get emote_is_active()))
   > Set frontEnd_thigh_alpha(frontEnd_thigh_alpha:double=1) -> (Output_Get:double)
   > Set frontEnd_cloth_alpha(frontEnd_cloth_alpha:double=1) -> (Output_Get:double)
   Set frontEnd_thigh_alpha(execute=[@n33.else, @n34.else], frontEnd_thigh_alpha=0)
   > Set frontEnd_cloth_alpha(frontEnd_cloth_alpha=0)
   n35 = Get BodyMesh()
   Set arm_aim_up_r(
    execute=@n13.then_2
    arm_aim_up_r:double=GetAlphaFromReferenceTransformDelta(
     SkeletalMeshComponent=@n35
     BoneName=upperarm_r
     BaseName=spine_05
     DeltaComponentType=QuaternionTwist
     InputRange=(X=60,Y=100)
     TwistAxis="0,0,1"
    )
   ) -> (Output_Get:double)
   n36 = "Clamp (Float)"(
    Value=float * float(
     A=GetAlphaFromReferenceTransformDelta(
      SkeletalMeshComponent=@n35
      BoneName=upperarm_r
      BaseName=clavicle_r
      DeltaComponentType=QuaternionTwist
      InputRange=(X=0,Y=-90)
      TwistAxis="1,0,0"
     )
     B=GetAlphaFromReferenceTransformDelta(
      SkeletalMeshComponent=@n35
      BoneName=upperarm_r
      BaseName=clavicle_r
      DeltaComponentType=QuaternionTwist
      InputRange=(X=0,Y=80)
      TwistAxis="0,1,0"
     )
    )
    Min=0
    Max=1
   )
   > n37 = Branch(Condition="float <= float"!(A:double=@n36, B:double=0.1) -> (ReturnValue:bool) {is=title K2Node_CallFunction})
   > n38 = Branch(Condition=Get actively_skydiving() -> (actively_skydiving:bool))
   > Set shldr_r_roll_alpha(shldr_r_roll_alpha:double=0) -> (Output_Get:double)
   Set shldr_r_roll_alpha(execute=@n38.else, shldr_r_roll_alpha=@n36)
   n39 = Branch(execute=@n37.else, Condition=Get actively_skydiving())
   > Set shldr_r_roll_alpha(shldr_r_roll_alpha=0)
   Set shldr_r_roll_alpha(execute=@n39.else, shldr_r_roll_alpha=FInterpTo(Current=Get shldr_r_roll_alpha(), Target=@n36, DeltaTime=@n12.DeltaTimeX, InterpSpeed=1.5))
   n40 = Branch(execute=@n13.then_6, Condition=Get actively_skydiving())
   > n41 = Branch(Condition=Get bIsSkydiveFloating())
   > Set shoulder_pad_float(shoulder_pad_float:double=1) -> (Output_Get:double)
   Set shoulder_pad_float(execute=[@n40.else, @n41.else], shoulder_pad_float=0)
   n42 = Gravity Override State Function DEPRECATED(
    execute=@n13.then_3
    CharacterPartAnimInstance=Self-Reference()
    GameTimeInSeconds=GetGameTimeInSeconds()
    GravityOverrideParams=Get Gravity Override Params()
   )
   > Set GravityOverride_b(GravityOverride_b:Vector=@n42) -> (Output_Get:Vector)
   n43 = Get bIsPlayingEmote()
   > Set bIsInFrontEndNoEmote(bIsInFrontEndNoEmote:bool=AND Boolean!(A=Get bIsInFrontEnd(), B=NOT Boolean(A=@n43)) -> (ReturnValue)) -> (Output_Get:bool)
   n44 = Get BodyMesh()
   n45 = GetAlphaFromReferenceTransformDelta(
    SkeletalMeshComponent=@n44
    BoneName=thigh_l
    BaseName=pelvis
    DeltaComponentType=QuaternionTwist
    InputRange=(X=20,Y=100)
    TwistAxis="0,0,1"
   )
   Set thigh_rot_l(execute=@n13.then_1, thigh_rot_l=@n45)
   n46 = GetAlphaFromReferenceTransformDelta(
    SkeletalMeshComponent=@n44
    BoneName=thigh_r
    BaseName=pelvis
    DeltaComponentType=QuaternionTwist
    InputRange=(X=20,Y=100)
    TwistAxis="0,0,1"
   )
   > Set thigh_rot_r(thigh_rot_r=@n46)
   > Branch(Condition="float > float"(A=Get PawnSpeed2D(), B=100) {is=title K2Node_CallFunction})
   > Set thigh_rot_l(thigh_rot_l=float * float(A=@n45, B=0.35))
   > Set thigh_rot_r(thigh_rot_r=float * float(A=@n46, B=0.35))
   n47 = Get BodyMesh()
   Set knee_rot_alpha_l(
    execute=@n13.then_4
    knee_rot_alpha_l:double=GetAlphaFromReferenceTransformDelta(
     SkeletalMeshComponent=@n47
     BoneName=calf_l
     BaseName=thigh_l
     DeltaComponentType=QuaternionTwist
     InputRange=(X=-25,Y=-110)
     TwistAxis="0,0,1"
    )
   ) -> (Output_Get:double)
   > Set knee_rot_alpha_r(knee_rot_alpha_r:double=GetAlphaFromReferenceTransformDelta(
     SkeletalMeshComponent=@n47
     BoneName=calf_r
     BaseName=thigh_r
     DeltaComponentType=QuaternionTwist
     InputRange=(X=-25,Y=-110)
     TwistAxis="0,0,1"
    )) -> (Output_Get:double)
   Event OnBeginPlayingEmote() {is=title K2Node_Event}
   > Set emote_is_active(emote_is_active:bool=true) -> (Output_Get:bool)
   Event OnEndPlayingEmote() {is=title K2Node_Event}
   > Set emote_is_active(emote_is_active=false)
   n48 = Event StartPreviewReactiveResponse!() -> (
    OutputDelegate
    then
    ChannelAndVariant:"Cosmetic Variant Info Structure (by ref)"
    Preview:"Fort Cosmetic Reactive Preview (F CRP) Object Reference"
   ) {is=title K2Node_Event}
   > n49 = Gate Is Cosmetic Variant to React Towards(CosmeticVariantInfo=@n48.ChannelAndVariant)
   Branch(Condition=NOT Boolean(A=Get changed_swimming() -> (changed_swimming:bool)))
   > n50 = Branch(Condition=OR Boolean!(A=Get bIsSurfaceSwimming(), B=Get bIsSwimSprinting()) -> (ReturnValue))
   > n51 = Branch(Condition=Get actively_swimming() -> (actively_swimming:bool))
   Set changed_swimming(execute=@n51.else, changed_swimming=true)
   > Set actively_swimming(actively_swimming:bool=true) -> (Output_Get:bool)
   Branch(execute=@n50.else, Condition=Get actively_swimming())
   n52 = Get bIsSwimSprinting()
   n53 = Get bIsSurfaceSwimming()
   > Branch(Condition=AND Boolean(A=NOT Boolean(A=@n53), B=NOT Boolean(A=@n52)))
   > Set actively_swimming(actively_swimming=false)
   > Set changed_swimming(changed_swimming=false)
   GetAlphaFromReferenceTransformDelta(
    SkeletalMeshComponent=@n35
    BoneName=upperarm_r
    BaseName=clavicle_r
    DeltaComponentType=QuaternionTwist
    InputRange=(X=0,Y=-75)
    TwistAxis="0,0,1"
   )
   "To String (Float)"(InDouble=0)
   Reroute Node(InputPin:Wildcard) -> (OutputPin:Wildcard)
   Set shldr_r_pouch_alpha(shldr_r_pouch_alpha:double=0) -> (Output_Get:double)
  ]}
 Gate_IsCosmeticVariantToReactTowards {nodes [
   n1 = Inputs!() -> (execute:Exec, CosmeticVariantInfo:"Cosmetic Variant Info Structure (by ref)")
   n2 = Break Cosmetic Variant Info!(CosmeticVariantInfo:"Cosmetic Variant Info Structure (by ref)"=@n1) -> (VariantChannelTag:Gameplay Tag Structure, ActiveVariantTag:Gameplay Tag Structure) {is=title K2Node_BreakStruct}
   > Branch(Condition=AND Boolean(
     A="Equal (GameplayTag)"!(A:Gameplay Tag Structure=@n2, B:Gameplay Tag Structure=(TagName="Cosmetics.Variant.Channel.Particle")) -> (ReturnValue:bool) {is=title K2Node_CallFunction}
     B="Equal (GameplayTag)"(A=@n2.ActiveVariantTag, B=(TagName="Cosmetics.Variant.Property.Particle1"))
    ))
   > Outputs!(then:Exec)
  ]}
 TailOff {nodes [TailOff() {is=title K2Node_FunctionEntry}]}
 TailOn {nodes [TailOn() {is=title K2Node_FunctionEntry}]}
]
variables [
 actively_skydiving {type=boolean}
 actively_swimming {type=boolean}
 arm_aim_up_r {type=number}
 arm_out_up_r {type=number}
 armor_offsetX {type=number, default=0.7}
 base_not_active {type=boolean}
 bIsInFrontEndNoEmote {type=boolean}
 bIsPreviewReacting_Swimming {type=boolean}
 BodyMesh {type=SkeletalMeshComponent}
 buckle_alpha {type=number, default=1}
 changed_swimming {type=boolean}
 cloth_alpha {type=number, default=1}
 cloth_light_alpha {type=number, default=1}
 CosmeticShoesMesh {type=SkeletalMeshComponent}
 crouch_alpha {type=number}
 disable_tail {type=boolean}
 emote_is_active {type=boolean}
 frontEnd_cloth_alpha {type=number}
 frontEnd_thigh_alpha {type=number}
 Gravity Override Params {
  type=GravityOverrideParamsStruct
  default {
   skydive_FloatingDown {gravityOverride=<0 0 0>}
   skydive_GoingUp {gravityOverride=<0 0 -100>}
   skydive_Parachuting {gravityOverride=<0 0 0>}
   swimming {gravityOverride=<0 0 0>}
  }
 }
 GravityOverride_b {type=Vector}
 JumpBootsMesh {type=SkeletalMeshComponent}
 knee_rot_alpha_l {type=number}
 knee_rot_alpha_r {type=number}
 light_armor_alpha {type=number}
 Mat_Instance {type=MaterialInstanceDynamic}
 PawnOwner {type=PlayerPawn_Athena_C}
 rear_cloth_alpha {type=number, default=1}
 shldr_r_pouch_alpha {type=number}
 shldr_r_roll_alpha {type=number}
 shoulder_pad_float {type=number, default=1}
 skydiving_arm_up {type=number}
 skydiving_opened_arm {type=number}
 TailMesh {type=SkeletalMeshComponent}
 TailOff {default=()}
 TailOn {default=()}
 thigh_rot_l {type=number}
 thigh_rot_r {type=number}
]
