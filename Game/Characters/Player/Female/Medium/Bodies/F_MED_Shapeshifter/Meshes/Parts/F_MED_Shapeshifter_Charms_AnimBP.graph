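# Anim Blueprint for the female Shapeshifter charms mesh part, animating its tail through swimming, sprinting and water jumps.
# It copies the attached parent's pose and layers a swim/jump state machine from proto_tail_01, whose rotation is then overridden.
# Additive state machines and a turn blendspace add directional and sprint tail motion.
# The cached result gains pitch and pose additives during dive jumps, or a fixed tail yaw offset in the front end.
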AnimBlueprint /Game/Characters/Player/Female/Medium/Bodies/F_MED_Shapeshifter/Meshes/Parts/F_MED_Shapeshifter_Charms_AnimBP
target_skeleton=&/F_MED_Shapeshifter_Charms_Skeleton
graphs [
 AnimGraph {nodes [
   n1 = Get SwimmingDivePitch() -> (SwimmingDivePitch:float)
   n2 = Apply Additive(
    Base=Apply Additive(
     Base=Use cached pose 'SavedPose11377'()
     Additive=Blendspace Player 'Shapeshifter_Swim_Dive_JumpPitch_BS'(X=0, Y=@n1) {blendSpace=/Game/Animation/Game/MainPlayer/Swimming/Shapeshifter/Shapeshifter_Swim_Dive_JumpPitch_BS}
    ) {alphaBoolBlend {blendOption=Linear}}
    Additive=Blendspace Player 'Shapeshifter_Swim_Dive_JumpPose_BS'(
     X=Get SwimmingYawDeltaRatePerSecond() -> (SwimmingYawDeltaRatePerSecond:float)
     Y=Get SwimmingDivePitchDeltaRatePerSec() -> (SwimmingDivePitchDeltaRatePerSec:float)
    ) {blendSpace=/Game/Animation/Game/MainPlayer/Swimming/Shapeshifter/Shapeshifter_Swim_Dive_JumpPose_BS}
   ) {alphaBoolBlend {blendOption=Linear}}
   n3 = Use cached pose 'SavedPose11377'()
   n4 = Get bIsDiveJumping() -> (bIsDiveJumping:bool)
   n5 = "Transform (Modify) Bone - Bone: proto_tail_01"(
    ComponentPose=Local To Component(LocalPose=Use cached pose 'SavedPose11377'())
    Alpha=1
    Translation="0,0,0"
    Rotation="0,-45,0"
    Scale="1,1,1"
   ) {rotationMode=BMM_Additive, alphaBoolBlend {blendOption=Linear}, boneToModify {boneName=proto_tail_01}, rotation {yaw=-45}}
   Output Pose(Result=Blend Poses by bool(
     BlendPose_0=Component To Local(ComponentPose=@n5)
     BlendPose_1=Blend Poses by bool(BlendPose_0=@n2, BlendPose_1=@n3, BlendTime_0=0.4, BlendTime_1=0.4, bActiveValue=@n4) {blendType=Linear}
     BlendTime_0=0
     BlendTime_1=0
     bActiveValue=Get bIsInFrontEnd()
    ) {blendType=Linear}) {layerGroup=None}
   AnimGraphNode_StateMachine_0 = New State Machine() {is=title AnimGraphNode_StateMachine}
   n6 = Use cached pose 'CopyPoseFromMesh'()
   n7 = "Transform (Modify) Bone - Bone: proto_tail_01"(
    ComponentPose=Local To Component(LocalPose=Layered blend per bone(BasePose=@n6, BlendPoses_0=@AnimGraphNode_StateMachine_0) {layerSetup [{branchFilters [{boneName=proto_tail_01}]}]})
    Alpha=1
    Translation="0,0,0"
    Rotation="90,90,90"
    Scale="1,1,1"
   ) {rotationMode=BMM_Replace, alphaBoolBlend {blendOption=Linear}, boneToModify {boneName=proto_tail_01}, rotation=<rot 90 90 90>}
   AnimGraphNode_StateMachine_2 = Additive Tail Base Rot() {is=title AnimGraphNode_StateMachine}
   n8 = Apply Additive(Base=Component To Local(ComponentPose=@n7), Additive=@AnimGraphNode_StateMachine_2) {alphaBoolBlend {blendOption=Linear}}
   AnimGraphNode_StateMachine_1 = Additive Tail Base Rot_Sprint() {is=title AnimGraphNode_StateMachine}
   n9 = Get SwimmingSprintYawDeltaRatePerSecondSmooth() -> (SwimmingSprintYawDeltaRatePerSecondSmooth:float)
   Save cached pose 'SavedPose11377'(Pose=Apply Additive(
     Base=Apply Mesh Space Additive!(Base:Pose Link Structure=@n8, Additive:Pose Link Structure=@AnimGraphNode_StateMachine_1, Alpha:float=1) -> (Pose:Pose Link Structure) {is=title AnimGraphNode_ApplyMeshSpaceAdditive, alphaBoolBlend {blendOption=Linear}}
     Additive=Blendspace Player 'Shapeshifter_Swim_Sprint_Turn_BS'!(X=@n9) -> (Pose) {blendSpace=/Game/Animation/Game/MainPlayer/Swimming/Shapeshifter/Shapeshifter_Swim_Sprint_Turn_BS}
    ) {alphaBoolBlend {blendOption=Linear}}) {cacheName=SavedPose11377}
   Save cached pose 'CopyPoseFromMesh'(Pose=Copy Pose From Mesh() {bUseAttachedParent=true}) {cacheName=CopyPoseFromMesh}
  ]}
 New State Machine {
  entry=Swimming
  path=AnimGraph.AnimGraphNode_StateMachine_0.New State Machine
  ids=id
  nodes [
   AnimStateNode_0 = Locomotion() {is=title AnimStateNode}
   AnimStateNode_1 = Sprint_Start() {is=title AnimStateNode}
   AnimStateNode_2 = Sprint() {is=title AnimStateNode}
   AnimStateNode_3 = Sprint_Stop() {is=title AnimStateNode}
   AnimStateNode_4 = Swimming() {is=title AnimStateNode}
   AnimStateNode_5 = Jump() {is=title AnimStateNode}
   AnimStateNode_6 = Fall() {is=title AnimStateNode}
   AnimStateNode_7 = WaterLand() {is=title AnimStateNode}
   AnimStateTransitionNode_0 = AnimStateTransitionNode()
   AnimStateTransitionNode_1 = AnimStateTransitionNode() {bAutomaticRuleBasedOnSequencePlayerInState=true}
   AnimStateTransitionNode_10 = AnimStateTransitionNode()
   AnimStateTransitionNode_11 = AnimStateTransitionNode()
   AnimStateTransitionNode_12 = AnimStateTransitionNode()
   AnimStateTransitionNode_15 = AnimStateTransitionNode() {crossfadeDuration=0.5}
   AnimStateTransitionNode_2 = AnimStateTransitionNode() {crossfadeDuration=0.05}
   AnimStateTransitionNode_3 = AnimStateTransitionNode()
   AnimStateTransitionNode_4 = AnimStateTransitionNode() {bAutomaticRuleBasedOnSequencePlayerInState=true}
   AnimStateTransitionNode_6 = AnimStateTransitionNode()
   AnimStateTransitionNode_7 = AnimStateTransitionNode()
   AnimStateTransitionNode_8 = AnimStateTransitionNode()
   AnimStateTransitionNode_9 = AnimStateTransitionNode() {bAutomaticRuleBasedOnSequencePlayerInState=true}
  ]
  links [
   {
    dst=WaterLand
    rule=AnimGraph.AnimGraphNode_StateMachine_0.New State Machine.AnimStateTransitionNode_11.Transition
    src=Fall
    via=AnimStateTransitionNode_11
   }
   {
    dst=WaterLand
    rule=AnimGraph.AnimGraphNode_StateMachine_0.New State Machine.AnimStateTransitionNode_7.Transition
    src=Fall
    via=AnimStateTransitionNode_7
   }
   {
    dst=WaterLand
    rule=AnimGraph.AnimGraphNode_StateMachine_0.New State Machine.AnimStateTransitionNode_8.Transition
    src=Fall
    via=AnimStateTransitionNode_8
   }
   {
    dst=Fall
    rule=AnimGraph.AnimGraphNode_StateMachine_0.New State Machine.AnimStateTransitionNode_10.Transition
    src=Jump
    via=AnimStateTransitionNode_10
   }
   {
    dst=Fall
    rule=AnimGraph.AnimGraphNode_StateMachine_0.New State Machine.AnimStateTransitionNode_6.Transition
    src=Jump
    via=AnimStateTransitionNode_6
   }
   {
    dst=Sprint_Start
    rule=AnimGraph.AnimGraphNode_StateMachine_0.New State Machine.AnimStateTransitionNode_0.AnimationTransitionGraph_0
    src=Locomotion
    via=AnimStateTransitionNode_0
   }
   {
    dst=Sprint_Stop
    rule=AnimGraph.AnimGraphNode_StateMachine_0.New State Machine.AnimStateTransitionNode_3.AnimationTransitionGraph_0
    src=Sprint
    via=AnimStateTransitionNode_3
   }
   {
    dst=Sprint
    rule=AnimGraph.AnimGraphNode_StateMachine_0.New State Machine.AnimStateTransitionNode_1.AnimationTransitionGraph_0
    src=Sprint_Start
    via=AnimStateTransitionNode_1
   }
   {
    dst=Sprint_Stop
    rule=AnimGraph.AnimGraphNode_StateMachine_0.New State Machine.AnimStateTransitionNode_15.AnimationTransitionGraph_0
    src=Sprint_Start
    via=AnimStateTransitionNode_15
   }
   {
    dst=Locomotion
    rule=AnimGraph.AnimGraphNode_StateMachine_0.New State Machine.AnimStateTransitionNode_4.AnimationTransitionGraph_0
    src=Sprint_Stop
    via=AnimStateTransitionNode_4
   }
   {
    dst=Jump
    rule=AnimGraph.AnimGraphNode_StateMachine_0.New State Machine.AnimStateTransitionNode_2.Transition
    src=Swimming
    via=AnimStateTransitionNode_2
   }
   {
    dst=Swimming
    rule=AnimGraph.AnimGraphNode_StateMachine_0.New State Machine.AnimStateTransitionNode_12.Transition
    src=WaterLand
    via=AnimStateTransitionNode_12
   }
   {
    dst=Swimming
    rule=AnimGraph.AnimGraphNode_StateMachine_0.New State Machine.AnimStateTransitionNode_9.Transition
    src=WaterLand
    via=AnimStateTransitionNode_9
   }
  ]
 }
 Locomotion {
  path=AnimGraph.AnimGraphNode_StateMachine_0.New State Machine.AnimStateNode_0.Locomotion
  nodes [
   AnimGraphNode_StateMachine_0 = Locomotion() {is=title AnimGraphNode_StateMachine}
   Output Animation Pose(Result=@AnimGraphNode_StateMachine_0)
  ]
 }
 Locomotion {
  entry=Idle
  path=AnimGraph.AnimGraphNode_StateMachine_0.New State Machine.AnimStateNode_0.Locomotion.AnimGraphNode_StateMachine_0.Locomotion
  ids=id
  nodes [
   AnimStateNode_0 = Idle()
   AnimStateNode_1 = DirectionalMovement() {is=title AnimStateNode}
   AnimStateTransitionNode_0 = AnimStateTransitionNode()
   AnimStateTransitionNode_1 = AnimStateTransitionNode()
  ]
  links [
   {
    dst=Idle
    rule=AnimGraph.AnimGraphNode_StateMachine_0.New State Machine.AnimStateNode_0.Locomotion.AnimGraphNode_StateMachine_0.Locomotion.AnimStateTransitionNode_1.Transition
    src=DirectionalMovement
    via=AnimStateTransitionNode_1
   }
   {
    dst=DirectionalMovement
    rule=AnimGraph.AnimGraphNode_StateMachine_0.New State Machine.AnimStateNode_0.Locomotion.AnimGraphNode_StateMachine_0.Locomotion.AnimStateTransitionNode_0.Transition
    src=Idle
    via=AnimStateTransitionNode_0
   }
  ]
 }
 Idle {
  path=AnimGraph.AnimGraphNode_StateMachine_0.New State Machine.AnimStateNode_0.Locomotion.AnimGraphNode_StateMachine_0.Locomotion.AnimStateNode_0.Idle
  nodes [
   Apply Additive(Additive="Sequence Player (additive) 'Shapeshifter_Swim_Tail_Direction_Rot_Idle'"() {sequence=/Game/Animation/Game/MainPlayer/Swimming/Shapeshifter/Shapeshifter_Swim_Tail_Direction_Rot_Idle}) {alphaBoolBlend {blendOption=Linear}}
   Output Animation Pose(Result=Sequence Player 'Shapeshifter_Swim_Idle_Noise'() {sequence=/Game/Animation/Game/MainPlayer/Swimming/Shapeshifter/Shapeshifter_Swim_Idle_Noise})
  ]
 }
 DirectionalMovement {
  path=AnimGraph.AnimGraphNode_StateMachine_0.New State Machine.AnimStateNode_0.Locomotion.AnimGraphNode_StateMachine_0.Locomotion.AnimStateNode_1.DirectionalMovement
  nodes [
   Apply Additive(Additive=Blendspace Player 'Shapeshifter_Swim_Sprint_Turn_BS'(X=Get SwimmingVelocityAngle() -> (SwimmingVelocityAngle:float)) {blendSpace=&Shapeshifter_Swim_Sprint_Turn_BS}) {alphaBoolBlend {blendOption=Linear}}
   Output Animation Pose(Result=Sequence Player 'Shapeshifter_Swim_N_Targeted'() {sequence=/Game/Animation/Game/MainPlayer/Swimming/Shapeshifter/Shapeshifter_Swim_N_Targeted})
  ]
 }
 Transition {
  path=AnimGraph.AnimGraphNode_StateMachine_0.New State Machine.AnimStateNode_0.Locomotion.AnimGraphNode_StateMachine_0.Locomotion.AnimStateTransitionNode_0.Transition
  nodes [Result(bCanEnterTransition="float > float"(A=Get Speed2D() -> (Speed2D:float), B=0) {is=title K2Node_CallFunction})]
 }
 Transition {
  path=AnimGraph.AnimGraphNode_StateMachine_0.New State Machine.AnimStateNode_0.Locomotion.AnimGraphNode_StateMachine_0.Locomotion.AnimStateTransitionNode_1.Transition
  nodes [Result(bCanEnterTransition="Equal (Float)"!(A:double=Get Speed2D(), B:double=0) -> (ReturnValue:bool) {is=title K2Node_CallFunction})]
 }
 Sprint_Start {
  path=AnimGraph.AnimGraphNode_StateMachine_0.New State Machine.AnimStateNode_1.Sprint_Start
  nodes [Output Animation Pose(Result=Sequence Player 'Shapeshifter_Swim_Sprint_Start'!(bLoopAnimation:bool=False) -> (Pose) {bLoopAnimation=false, sequence=/Game/Animation/Game/MainPlayer/Swimming/Shapeshifter/Shapeshifter_Swim_Sprint_Start})]
 }
 Sprint {
  path=AnimGraph.AnimGraphNode_StateMachine_0.New State Machine.AnimStateNode_2.Sprint
  nodes [Output Animation Pose(Result=Blend Poses by bool(
     BlendPose_0=Local Space Ref Pose()
     BlendPose_1=Sequence Player 'Shapeshifter_Swim_Sprint'() {sequence=/Game/Animation/Game/MainPlayer/Swimming/Shapeshifter/Shapeshifter_Swim_Sprint}
     bActiveValue=Get bIsDiveJumping()
    ) {blendType=Linear})]
 }
 Sprint_Stop {
  path=AnimGraph.AnimGraphNode_StateMachine_0.New State Machine.AnimStateNode_3.Sprint_Stop
  nodes [Output Animation Pose(Result=Sequence Player 'Shapeshifter_Swim_Sprint_Stop'!(bLoopAnimation:bool=False) -> (Pose) {bLoopAnimation=false, sequence=/Game/Animation/Game/MainPlayer/Swimming/Shapeshifter/Shapeshifter_Swim_Sprint_Stop})]
 }
 Swimming {
  path=AnimGraph.AnimGraphNode_StateMachine_0.New State Machine.AnimStateNode_4.Swimming
  nodes [
   AnimGraphNode_StateMachine_0 = Swimming() {is=title AnimGraphNode_StateMachine}
   Output Animation Pose(Result=@AnimGraphNode_StateMachine_0)
  ]
 }
 Swimming {
  entry=Locomotion
  path=AnimGraph.AnimGraphNode_StateMachine_0.New State Machine.AnimStateNode_4.Swimming.AnimGraphNode_StateMachine_0.Swimming
  ids=id
  nodes [
   AnimStateNode_0 = Locomotion() {is=title AnimStateNode}
   AnimStateNode_1 = Sprint_Start()
   AnimStateNode_2 = Sprint()
   AnimStateNode_3 = Sprint_Stop()
   AnimStateTransitionNode_0 = AnimStateTransitionNode()
   AnimStateTransitionNode_1 = AnimStateTransitionNode() {bAutomaticRuleBasedOnSequencePlayerInState=true}
   AnimStateTransitionNode_15 = AnimStateTransitionNode() {crossfadeDuration=0.5}
   AnimStateTransitionNode_3 = AnimStateTransitionNode()
   AnimStateTransitionNode_4 = AnimStateTransitionNode() {bAutomaticRuleBasedOnSequencePlayerInState=true}
  ]
  links [
   {
    dst=Sprint_Start
    rule=AnimGraph.AnimGraphNode_StateMachine_0.New State Machine.AnimStateNode_4.Swimming.AnimGraphNode_StateMachine_0.Swimming.AnimStateTransitionNode_0.AnimationTransitionGraph_0
    src=Locomotion
    via=AnimStateTransitionNode_0
   }
   {
    dst=Sprint_Stop
    rule=AnimGraph.AnimGraphNode_StateMachine_0.New State Machine.AnimStateNode_4.Swimming.AnimGraphNode_StateMachine_0.Swimming.AnimStateTransitionNode_3.AnimationTransitionGraph_0
    src=Sprint
    via=AnimStateTransitionNode_3
   }
   {
    dst=Sprint
    rule=AnimGraph.AnimGraphNode_StateMachine_0.New State Machine.AnimStateNode_4.Swimming.AnimGraphNode_StateMachine_0.Swimming.AnimStateTransitionNode_1.AnimationTransitionGraph_0
    src=Sprint_Start
    via=AnimStateTransitionNode_1
   }
   {
    dst=Sprint_Stop
    rule=AnimGraph.AnimGraphNode_StateMachine_0.New State Machine.AnimStateNode_4.Swimming.AnimGraphNode_StateMachine_0.Swimming.AnimStateTransitionNode_15.AnimationTransitionGraph_0
    src=Sprint_Start
    via=AnimStateTransitionNode_15
   }
   {
    dst=Locomotion
    rule=AnimGraph.AnimGraphNode_StateMachine_0.New State Machine.AnimStateNode_4.Swimming.AnimGraphNode_StateMachine_0.Swimming.AnimStateTransitionNode_4.AnimationTransitionGraph_0
    src=Sprint_Stop
    via=AnimStateTransitionNode_4
   }
  ]
 }
 Locomotion {
  path=AnimGraph.AnimGraphNode_StateMachine_0.New State Machine.AnimStateNode_4.Swimming.AnimGraphNode_StateMachine_0.Swimming.AnimStateNode_0.Locomotion
  nodes [
   AnimGraphNode_StateMachine_0 = Locomotion() {is=title AnimGraphNode_StateMachine}
   Output Animation Pose(Result=@AnimGraphNode_StateMachine_0)
  ]
 }
 Locomotion {
  entry=Idle
  path=AnimGraph.AnimGraphNode_StateMachine_0.New State Machine.AnimStateNode_4.Swimming.AnimGraphNode_StateMachine_0.Swimming.AnimStateNode_0.Locomotion.AnimGraphNode_StateMachine_0.Locomotion
  ids=id
  nodes [
   AnimStateNode_0 = Idle()
   AnimStateNode_1 = DirectionalMovement()
   AnimStateTransitionNode_0 = AnimStateTransitionNode()
   AnimStateTransitionNode_1 = AnimStateTransitionNode()
  ]
  links [
   {
    dst=Idle
    rule=AnimGraph.AnimGraphNode_StateMachine_0.New State Machine.AnimStateNode_4.Swimming.AnimGraphNode_StateMachine_0.Swimming.AnimStateNode_0.Locomotion.AnimGraphNode_StateMachine_0.Locomotion.AnimStateTransitionNode_1.Transition
    src=DirectionalMovement
    via=AnimStateTransitionNode_1
   }
   {
    dst=DirectionalMovement
    rule=AnimGraph.AnimGraphNode_StateMachine_0.New State Machine.AnimStateNode_4.Swimming.AnimGraphNode_StateMachine_0.Swimming.AnimStateNode_0.Locomotion.AnimGraphNode_StateMachine_0.Locomotion.AnimStateTransitionNode_0.Transition
    src=Idle
    via=AnimStateTransitionNode_0
   }
  ]
 }
 Idle {
  path=AnimGraph.AnimGraphNode_StateMachine_0.New State Machine.AnimStateNode_4.Swimming.AnimGraphNode_StateMachine_0.Swimming.AnimStateNode_0.Locomotion.AnimGraphNode_StateMachine_0.Locomotion.AnimStateNode_0.Idle
  nodes [
   Apply Additive(Additive="Sequence Player (additive) 'Shapeshifter_Swim_Tail_Direction_Rot_Idle'"() {sequence=&Shapeshifter_Swim_Tail_Direction_Rot_Idle}) {alphaBoolBlend {blendOption=Linear}}
   Output Animation Pose(Result=Sequence Player 'Shapeshifter_Swim_Idle_Noise'() {sequence=&Shapeshifter_Swim_Idle_Noise})
  ]
 }
 DirectionalMovement {
  path=AnimGraph.AnimGraphNode_StateMachine_0.New State Machine.AnimStateNode_4.Swimming.AnimGraphNode_StateMachine_0.Swimming.AnimStateNode_0.Locomotion.AnimGraphNode_StateMachine_0.Locomotion.AnimStateNode_1.DirectionalMovement
  nodes [
   Apply Additive(Additive=Blendspace Player 'Shapeshifter_Swim_Sprint_Turn_BS'(X=Get SwimmingVelocityAngle()) {blendSpace=&Shapeshifter_Swim_Sprint_Turn_BS}) {alphaBoolBlend {blendOption=Linear}}
   Output Animation Pose(Result=Sequence Player 'Shapeshifter_Swim_N_Targeted'() {sequence=&Shapeshifter_Swim_N_Targeted})
  ]
 }
 Transition {
  path=AnimGraph.AnimGraphNode_StateMachine_0.New State Machine.AnimStateNode_4.Swimming.AnimGraphNode_StateMachine_0.Swimming.AnimStateNode_0.Locomotion.AnimGraphNode_StateMachine_0.Locomotion.AnimStateTransitionNode_0.Transition
  nodes [Result(bCanEnterTransition="float > float"(A=Get Speed2D(), B=0))]
 }
 Transition {
  path=AnimGraph.AnimGraphNode_StateMachine_0.New State Machine.AnimStateNode_4.Swimming.AnimGraphNode_StateMachine_0.Swimming.AnimStateNode_0.Locomotion.AnimGraphNode_StateMachine_0.Locomotion.AnimStateTransitionNode_1.Transition
  nodes [Result(bCanEnterTransition="Equal (Float)"(A=Get Speed2D(), B=0))]
 }
 Sprint_Start {
  path=AnimGraph.AnimGraphNode_StateMachine_0.New State Machine.AnimStateNode_4.Swimming.AnimGraphNode_StateMachine_0.Swimming.AnimStateNode_1.Sprint_Start
  nodes [Output Animation Pose(Result=Sequence Player 'Shapeshifter_Swim_Sprint_Start'(bLoopAnimation=False) {bLoopAnimation=false, sequence=&Shapeshifter_Swim_Sprint_Start})]
 }
 Sprint {
  path=AnimGraph.AnimGraphNode_StateMachine_0.New State Machine.AnimStateNode_4.Swimming.AnimGraphNode_StateMachine_0.Swimming.AnimStateNode_2.Sprint
  nodes [Output Animation Pose(Result=Blend Poses by bool(
     BlendPose_0=Local Space Ref Pose()
     BlendPose_1=Sequence Player 'Shapeshifter_Swim_Sprint'() {sequence=&Shapeshifter_Swim_Sprint}
     bActiveValue=Get bIsDiveJumping()
    ) {blendType=Linear})]
 }
 Sprint_Stop {
  path=AnimGraph.AnimGraphNode_StateMachine_0.New State Machine.AnimStateNode_4.Swimming.AnimGraphNode_StateMachine_0.Swimming.AnimStateNode_3.Sprint_Stop
  nodes [Output Animation Pose(Result=Sequence Player 'Shapeshifter_Swim_Sprint_Stop'(bLoopAnimation=False) {bLoopAnimation=false, sequence=&Shapeshifter_Swim_Sprint_Stop})]
 }
 AnimationTransitionGraph_0 {
  path=AnimGraph.AnimGraphNode_StateMachine_0.New State Machine.AnimStateNode_4.Swimming.AnimGraphNode_StateMachine_0.Swimming.AnimStateTransitionNode_0.AnimationTransitionGraph_0
  nodes [Result(bCanEnterTransition=Get bIsSprinting())]
 }
 AnimationTransitionGraph_0 {
  path=AnimGraph.AnimGraphNode_StateMachine_0.New State Machine.AnimStateNode_4.Swimming.AnimGraphNode_StateMachine_0.Swimming.AnimStateTransitionNode_1.AnimationTransitionGraph_0
  nodes [Result()]
 }
 AnimationTransitionGraph_0 {
  path=AnimGraph.AnimGraphNode_StateMachine_0.New State Machine.AnimStateNode_4.Swimming.AnimGraphNode_StateMachine_0.Swimming.AnimStateTransitionNode_15.AnimationTransitionGraph_0
  nodes [Result(bCanEnterTransition=NOT Boolean(A=Get bIsSprinting()))]
 }
 AnimationTransitionGraph_0 {
  path=AnimGraph.AnimGraphNode_StateMachine_0.New State Machine.AnimStateNode_4.Swimming.AnimGraphNode_StateMachine_0.Swimming.AnimStateTransitionNode_3.AnimationTransitionGraph_0
  nodes [Result(bCanEnterTransition=NOT Boolean(A=Get bIsSprinting()))]
 }
 AnimationTransitionGraph_0 {
  path=AnimGraph.AnimGraphNode_StateMachine_0.New State Machine.AnimStateNode_4.Swimming.AnimGraphNode_StateMachine_0.Swimming.AnimStateTransitionNode_4.AnimationTransitionGraph_0
  nodes [Result()]
 }
 Jump {
  path=AnimGraph.AnimGraphNode_StateMachine_0.New State Machine.AnimStateNode_5.Jump
  nodes [Output Animation Pose(Result=Sequence Player 'Shapeshifter_Swim_Jump_Loop'() {sequence=/Game/Animation/Game/MainPlayer/Swimming/Shapeshifter/Shapeshifter_Swim_Jump_Loop})]
 }
 Fall {
  path=AnimGraph.AnimGraphNode_StateMachine_0.New State Machine.AnimStateNode_6.Fall
  nodes [Output Animation Pose(Result=Sequence Player 'Shapeshifter_Swim_Jump_Fall'!(bLoopAnimation:bool=False) -> (Pose) {bLoopAnimation=false, sequence=/Game/Animation/Game/MainPlayer/Swimming/Shapeshifter/Shapeshifter_Swim_Jump_Fall})]
 }
 WaterLand {
  path=AnimGraph.AnimGraphNode_StateMachine_0.New State Machine.AnimStateNode_7.WaterLand
  nodes [Output Animation Pose(Result=Sequence Player 'Shapeshifter_Swim_Jump_Land'!(bLoopAnimation:bool=False) -> (Pose) {bLoopAnimation=false, sequence=/Game/Animation/Game/MainPlayer/Swimming/Shapeshifter/Shapeshifter_Swim_Jump_Land})]
 }
 AnimationTransitionGraph_0 {
  path=AnimGraph.AnimGraphNode_StateMachine_0.New State Machine.AnimStateTransitionNode_0.AnimationTransitionGraph_0
  nodes [Result(bCanEnterTransition=Get bIsSprinting())]
 }
 AnimationTransitionGraph_0 {path=AnimGraph.AnimGraphNode_StateMachine_0.New State Machine.AnimStateTransitionNode_1.AnimationTransitionGraph_0, nodes [Result()]}
 Transition {
  path=AnimGraph.AnimGraphNode_StateMachine_0.New State Machine.AnimStateTransitionNode_10.Transition
  nodes [Result(bCanEnterTransition=Get bIsSprinting())]
 }
 Transition {
  path=AnimGraph.AnimGraphNode_StateMachine_0.New State Machine.AnimStateTransitionNode_11.Transition
  nodes [Result(bCanEnterTransition=Get bIsSprinting())]
 }
 Transition {
  path=AnimGraph.AnimGraphNode_StateMachine_0.New State Machine.AnimStateTransitionNode_12.Transition
  nodes [Result(bCanEnterTransition=Get bIsSprinting())]
 }
 AnimationTransitionGraph_0 {
  path=AnimGraph.AnimGraphNode_StateMachine_0.New State Machine.AnimStateTransitionNode_15.AnimationTransitionGraph_0
  nodes [Result(bCanEnterTransition=NOT Boolean(A=Get bIsSprinting()))]
 }
 Transition {
  path=AnimGraph.AnimGraphNode_StateMachine_0.New State Machine.AnimStateTransitionNode_2.Transition
  nodes [Result(bCanEnterTransition=Get bTransition_IdleAdditive_to_SwimJumpStart() -> (bTransition_IdleAdditive_to_SwimJumpStart:bool))]
 }
 AnimationTransitionGraph_0 {
  path=AnimGraph.AnimGraphNode_StateMachine_0.New State Machine.AnimStateTransitionNode_3.AnimationTransitionGraph_0
  nodes [Result(bCanEnterTransition=NOT Boolean(A=Get bIsSprinting()))]
 }
 AnimationTransitionGraph_0 {path=AnimGraph.AnimGraphNode_StateMachine_0.New State Machine.AnimStateTransitionNode_4.AnimationTransitionGraph_0, nodes [Result()]}
 Transition {
  path=AnimGraph.AnimGraphNode_StateMachine_0.New State Machine.AnimStateTransitionNode_6.Transition
  nodes [Result(bCanEnterTransition=NOT Boolean(A=Get bIsSurfaceSwimming()))]
 }
 Transition {
  path=AnimGraph.AnimGraphNode_StateMachine_0.New State Machine.AnimStateTransitionNode_7.Transition
  nodes [Result(bCanEnterTransition=Get bIsSurfaceSwimming())]
 }
 Transition {
  path=AnimGraph.AnimGraphNode_StateMachine_0.New State Machine.AnimStateTransitionNode_8.Transition
  nodes [Result(bCanEnterTransition=Get bIsOnGround() -> (bIsOnGround:bool))]
 }
 Transition {path=AnimGraph.AnimGraphNode_StateMachine_0.New State Machine.AnimStateTransitionNode_9.Transition, nodes [Result()]}
 Additive Tail Base Rot_Sprint {
  entry=Additive Locomotion
  path=AnimGraph.AnimGraphNode_StateMachine_1.Additive Tail Base Rot_Sprint
  ids=id
  nodes [
   AnimStateNode_0 = AdditiveSprintStart() {is=title AnimStateNode}
   AnimStateNode_1 = AdditiveSprintLoop() {is=title AnimStateNode}
   AnimStateNode_2 = AdditiveSprintStop() {is=title AnimStateNode}
   AnimStateNode_3 = Additive Locomotion() {is=title AnimStateNode}
   AnimStateNode_5 = AdditiveSprintSlowPose() {is=title AnimStateNode}
   AnimStateTransitionNode_0 = AnimStateTransitionNode() {bAutomaticRuleBasedOnSequencePlayerInState=true, crossfadeDuration=0.05}
   AnimStateTransitionNode_1 = AnimStateTransitionNode()
   AnimStateTransitionNode_10 = AnimStateTransitionNode()
   AnimStateTransitionNode_11 = AnimStateTransitionNode()
   AnimStateTransitionNode_12 = AnimStateTransitionNode()
   AnimStateTransitionNode_16 = AnimStateTransitionNode() {crossfadeDuration=0.5}
   AnimStateTransitionNode_3 = AnimStateTransitionNode() {bAutomaticRuleBasedOnSequencePlayerInState=true}
   AnimStateTransitionNode_6 = AnimStateTransitionNode() {crossfadeDuration=0.4}
   AnimStateTransitionNode_7 = AnimStateTransitionNode() {crossfadeDuration=0.4}
   AnimStateTransitionNode_8 = AnimStateTransitionNode()
  ]
  links [
   {
    dst=AdditiveSprintStart
    rule=AnimGraph.AnimGraphNode_StateMachine_1.Additive Tail Base Rot_Sprint.AnimStateTransitionNode_10.Transition
    src=Additive Locomotion
    via=AnimStateTransitionNode_10
   }
   {
    dst=AdditiveSprintSlowPose
    rule=AnimGraph.AnimGraphNode_StateMachine_1.Additive Tail Base Rot_Sprint.AnimStateTransitionNode_6.Transition
    src=AdditiveSprintLoop
    via=AnimStateTransitionNode_6
   }
   {
    dst=AdditiveSprintStop
    rule=AnimGraph.AnimGraphNode_StateMachine_1.Additive Tail Base Rot_Sprint.AnimStateTransitionNode_11.Transition
    src=AdditiveSprintLoop
    via=AnimStateTransitionNode_11
   }
   {
    dst=AdditiveSprintLoop
    rule=AnimGraph.AnimGraphNode_StateMachine_1.Additive Tail Base Rot_Sprint.AnimStateTransitionNode_7.Transition
    src=AdditiveSprintSlowPose
    via=AnimStateTransitionNode_7
   }
   {
    dst=AdditiveSprintStart
    rule=AnimGraph.AnimGraphNode_StateMachine_1.Additive Tail Base Rot_Sprint.AnimStateTransitionNode_8.Transition
    src=AdditiveSprintSlowPose
    via=AnimStateTransitionNode_8
   }
   {
    dst=AdditiveSprintStop
    rule=AnimGraph.AnimGraphNode_StateMachine_1.Additive Tail Base Rot_Sprint.AnimStateTransitionNode_12.Transition
    src=AdditiveSprintSlowPose
    via=AnimStateTransitionNode_12
   }
   {
    dst=AdditiveSprintLoop
    rule=AnimGraph.AnimGraphNode_StateMachine_1.Additive Tail Base Rot_Sprint.AnimStateTransitionNode_0.AnimationTransitionGraph_0
    src=AdditiveSprintStart
    via=AnimStateTransitionNode_0
   }
   {
    dst=AdditiveSprintStop
    rule=AnimGraph.AnimGraphNode_StateMachine_1.Additive Tail Base Rot_Sprint.AnimStateTransitionNode_16.Transition
    src=AdditiveSprintStart
    via=AnimStateTransitionNode_16
   }
   {
    dst=Additive Locomotion
    rule=AnimGraph.AnimGraphNode_StateMachine_1.Additive Tail Base Rot_Sprint.AnimStateTransitionNode_1.Transition
    src=AdditiveSprintStop
    via=AnimStateTransitionNode_1
   }
   {
    dst=Additive Locomotion
    rule=AnimGraph.AnimGraphNode_StateMachine_1.Additive Tail Base Rot_Sprint.AnimStateTransitionNode_3.Transition
    src=AdditiveSprintStop
    via=AnimStateTransitionNode_3
   }
  ]
 }
 AdditiveSprintStart {
  path=AnimGraph.AnimGraphNode_StateMachine_1.Additive Tail Base Rot_Sprint.AnimStateNode_0.AdditiveSprintStart
  nodes [Output Animation Pose(Result="Sequence Player (additive) 'Shapeshifter_Swim_Sprint_Start_Tail_Add'"!(bLoopAnimation:bool=False) -> (Pose) {bLoopAnimation=false, sequence=/Game/Animation/Game/MainPlayer/Swimming/Shapeshifter/Shapeshifter_Swim_Sprint_Start_Tail_Add})]
 }
 AdditiveSprintLoop {
  path=AnimGraph.AnimGraphNode_StateMachine_1.Additive Tail Base Rot_Sprint.AnimStateNode_1.AdditiveSprintLoop
  nodes [Output Animation Pose(Result="Sequence Player (additive) 'Shapeshifter_Swim_Sprint_Pose_Tail_Add'"() {sequence=/Game/Animation/Game/MainPlayer/Swimming/Shapeshifter/Shapeshifter_Swim_Sprint_Pose_Tail_Add})]
 }
 AdditiveSprintStop {
  path=AnimGraph.AnimGraphNode_StateMachine_1.Additive Tail Base Rot_Sprint.AnimStateNode_2.AdditiveSprintStop
  nodes [Output Animation Pose(Result="Sequence Player (additive) 'Shapeshifter_Swim_Sprint_Stop_Tail_Add'"!(bLoopAnimation:bool=False) -> (Pose) {bLoopAnimation=false, sequence=/Game/Animation/Game/MainPlayer/Swimming/Shapeshifter/Shapeshifter_Swim_Sprint_Stop_Tail_Add})]
 }
 Additive Locomotion {
  path=AnimGraph.AnimGraphNode_StateMachine_1.Additive Tail Base Rot_Sprint.AnimStateNode_3.Additive Locomotion
  nodes [Output Animation Pose(Result=Additive Identity Pose())]
 }
 AdditiveSprintSlowPose {
  path=AnimGraph.AnimGraphNode_StateMachine_1.Additive Tail Base Rot_Sprint.AnimStateNode_5.AdditiveSprintSlowPose
  nodes [Output Animation Pose(Result="Sequence Player (additive) 'Shapeshifter_Swim_Sprint_Slow_Pose_Tail_Add'"() {sequence=/Game/Animation/Game/MainPlayer/Swimming/Shapeshifter/Shapeshifter_Swim_Sprint_Slow_Pose_Tail_Add})]
 }
 AnimationTransitionGraph_0 {
  path=AnimGraph.AnimGraphNode_StateMachine_1.Additive Tail Base Rot_Sprint.AnimStateTransitionNode_0.AnimationTransitionGraph_0
  nodes [Result()]
 }
 Transition {
  path=AnimGraph.AnimGraphNode_StateMachine_1.Additive Tail Base Rot_Sprint.AnimStateTransitionNode_1.Transition
  nodes [Result(bCanEnterTransition=Get bTransition_IdleAdditive_to_SwimJumpStart())]
 }
 Transition {
  path=AnimGraph.AnimGraphNode_StateMachine_1.Additive Tail Base Rot_Sprint.AnimStateTransitionNode_10.Transition
  nodes [Result(bCanEnterTransition=Get bIsSprinting())]
 }
 Transition {
  path=AnimGraph.AnimGraphNode_StateMachine_1.Additive Tail Base Rot_Sprint.AnimStateTransitionNode_11.Transition
  nodes [Result(bCanEnterTransition=NOT Boolean(A=Get bIsSprinting()))]
 }
 Transition {
  path=AnimGraph.AnimGraphNode_StateMachine_1.Additive Tail Base Rot_Sprint.AnimStateTransitionNode_12.Transition
  nodes [Result(bCanEnterTransition=NOT Boolean(A=Get bIsSprinting()))]
 }
 Transition {
  path=AnimGraph.AnimGraphNode_StateMachine_1.Additive Tail Base Rot_Sprint.AnimStateTransitionNode_16.Transition
  nodes [Result(bCanEnterTransition=NOT Boolean(A=Get bIsSprinting()))]
 }
 Transition {path=AnimGraph.AnimGraphNode_StateMachine_1.Additive Tail Base Rot_Sprint.AnimStateTransitionNode_3.Transition, nodes [Result()]}
 Transition {
  path=AnimGraph.AnimGraphNode_StateMachine_1.Additive Tail Base Rot_Sprint.AnimStateTransitionNode_6.Transition
  nodes [Result(bCanEnterTransition=Get bTransition_SwimSprint_to_SwimSprintSlow() -> (bTransition_SwimSprint_to_SwimSprintSlow:bool))]
 }
 Transition {
  path=AnimGraph.AnimGraphNode_StateMachine_1.Additive Tail Base Rot_Sprint.AnimStateTransitionNode_7.Transition
  nodes [Result(bCanEnterTransition=Get bTransition_SwimSprintSlow_to_SwimSprint() -> (bTransition_SwimSprintSlow_to_SwimSprint:bool))]
 }
 Transition {
  path=AnimGraph.AnimGraphNode_StateMachine_1.Additive Tail Base Rot_Sprint.AnimStateTransitionNode_8.Transition
  nodes [Result(bCanEnterTransition=Get bTransition_SwimSprintSlow_to_SwimSprintStart() -> (bTransition_SwimSprintSlow_to_SwimSprintStart:bool))]
 }
 Additive Tail Base Rot {
  entry=Additive Locomotion
  path=AnimGraph.AnimGraphNode_StateMachine_2.Additive Tail Base Rot
  ids=id
  nodes [AnimStateNode_0 = Additive Locomotion()]
 }
 Additive Locomotion {
  path=AnimGraph.AnimGraphNode_StateMachine_2.Additive Tail Base Rot.AnimStateNode_0.Additive Locomotion
  nodes [
   AnimGraphNode_StateMachine_0 = Additive Locomotion() {is=title AnimGraphNode_StateMachine}
   Output Animation Pose(Result=Blend Poses by bool(BlendPose_0=Additive Identity Pose(), BlendPose_1=@AnimGraphNode_StateMachine_0, bActiveValue=Get bIsSprinting()) {blendType=Linear})
  ]
 }
 Additive Locomotion {
  entry=Additive Idle
  path=AnimGraph.AnimGraphNode_StateMachine_2.Additive Tail Base Rot.AnimStateNode_0.Additive Locomotion.AnimGraphNode_StateMachine_0.Additive Locomotion
  ids=id
  nodes [
   AnimStateNode_0 = Additive Idle() {is=title AnimStateNode}
   AnimStateNode_1 = Additive DirectionalMovement() {is=title AnimStateNode}
   AnimStateTransitionNode_6 = AnimStateTransitionNode()
   AnimStateTransitionNode_7 = AnimStateTransitionNode() {crossfadeDuration=0.4}
  ]
  links [
   {
    dst=Additive Idle
    rule=AnimGraph.AnimGraphNode_StateMachine_2.Additive Tail Base Rot.AnimStateNode_0.Additive Locomotion.AnimGraphNode_StateMachine_0.Additive Locomotion.AnimStateTransitionNode_7.Transition
    src=Additive DirectionalMovement
    via=AnimStateTransitionNode_7
   }
   {
    dst=Additive DirectionalMovement
    rule=AnimGraph.AnimGraphNode_StateMachine_2.Additive Tail Base Rot.AnimStateNode_0.Additive Locomotion.AnimGraphNode_StateMachine_0.Additive Locomotion.AnimStateTransitionNode_6.Transition
    src=Additive Idle
    via=AnimStateTransitionNode_6
   }
  ]
 }
 Additive Idle {
  path=AnimGraph.AnimGraphNode_StateMachine_2.Additive Tail Base Rot.AnimStateNode_0.Additive Locomotion.AnimGraphNode_StateMachine_0.Additive Locomotion.AnimStateNode_0.Additive Idle
  nodes [Output Animation Pose(Result="Sequence Player (additive) 'Shapeshifter_Swim_Tail_Direction_Rot_Idle'"() {sequence=&Shapeshifter_Swim_Tail_Direction_Rot_Idle})]
 }
 Additive DirectionalMovement {
  path=AnimGraph.AnimGraphNode_StateMachine_2.Additive Tail Base Rot.AnimStateNode_0.Additive Locomotion.AnimGraphNode_StateMachine_0.Additive Locomotion.AnimStateNode_1.Additive DirectionalMovement
  nodes [Output Animation Pose(Result=Blendspace Player 'Shapeshifter_Swim_Tail_Dir_Rot_BS2'(X=Get SwimmingVelocityAngle(), Y=Get SwimmingSpeedRelativeToFlow() -> (SwimmingSpeedRelativeToFlow:float)) {blendSpace=/Game/Animation/Game/MainPlayer/Swimming/Shapeshifter/Shapeshifter_Swim_Tail_Dir_Rot_BS2})]
 }
 Transition {
  path=AnimGraph.AnimGraphNode_StateMachine_2.Additive Tail Base Rot.AnimStateNode_0.Additive Locomotion.AnimGraphNode_StateMachine_0.Additive Locomotion.AnimStateTransitionNode_6.Transition
  nodes [
   n1 = Get Speed2D()
   n2 = Get bIsSprinting()
   Result(bCanEnterTransition=AND Boolean!(A="float > float"(A=@n1, B=0), B=NOT Boolean(A=@n2), C:bool=Get bIsAccelerating2D()) -> (ReturnValue))
  ]
 }
 Transition {
  path=AnimGraph.AnimGraphNode_StateMachine_2.Additive Tail Base Rot.AnimStateNode_0.Additive Locomotion.AnimGraphNode_StateMachine_0.Additive Locomotion.AnimStateTransitionNode_7.Transition
  nodes [
   n1 = Get bIsAccelerating2D()
   n2 = Get Speed2D()
   Result(bCanEnterTransition=OR Boolean(A="Equal (Float)"(A=@n2, B=0), B=NOT Boolean(A=@n1)))
  ]
 }
 EventGraph {links [], nodes []}
]
variables [
 Delta Time {type=number}
 FortPlayerPawn_0 {type=FortPlayerPawn}
 PlayerAnimBP {type=Fortnite_M_Avg_Player_AnimBlueprint_C, default=null}
 SwimmingAnimBP {type=SwimmingAnimBP_C, default=null}
 TailRotationOverride {type=Rotator}
 TailRotationOverrideSmooth {type=Rotator}
]
